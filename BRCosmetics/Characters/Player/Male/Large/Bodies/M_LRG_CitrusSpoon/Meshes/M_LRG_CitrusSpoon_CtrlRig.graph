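# Control Rig for the M_LRG_CitrusSpoon large male body that drives corrective deform bones at knees, hips, elbows and upper arms.
# Knee and elbow bend angles remap to alphas, while spherical pose readers measure thigh and upper-arm poses.
# Those alphas weight transform offsets from a value struct, with right sides mirrored from left, often on base bones parent-constrained between neighbouring bones.
# Curve-struct flags optionally publish alphas as root animation attributes.
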ControlRigBlueprint /BRCosmetics/Characters/Player/Male/Large/Bodies/M_LRG_CitrusSpoon/Meshes/M_LRG_CitrusSpoon_CtrlRig
graphs [
 ContainedGraph {nodes [
   n1 = RigVMFunctionEntryNode()
   n2 = &RigUnit_GetTransform(Space=LocalSpace, Item=@n1.Item)
   n3 = &RigUnit_GetTransform(Space=LocalSpace, bInitial=true, Item=@n1.Item)
   > RigVMFunctionReturnNode(OutTransform=/Script/RigVM.RigVMFunction_MathTransformToEulerTransform(Value=&RigVMFunction_MathTransformMakeRelative(Global=@n2.Transform, Parent=@n3.Transform).Local) {is=struct RigVMUnitNode, template="Cast::Execute(in Value,out Result)"}.Result)
  ]}
 ContainedGraph {nodes [
   n1 = RigVMFunctionEntryNode()
   n2 = &RigVMFunction_MathTransformToEulerTransform(Value=@n1.transform) {template="Cast::Execute(in Value,out Result)"}
   > RigVMFunctionReturnNode(
    mirrorTransform=(Rotation=(X=0,Y=0,Z=0,W=1),Translation=(X=0,Y=0,Z=0),Scale3D=(X=0,Y=0,Z=0))
    mirrorTransform.Rotation=/Script/RigVM.RigVMFunction_MathQuaternionFromRotatorV2(Value=@n2.Result.Rotation) {is=struct RigVMUnitNode, template="Cast::Execute(in Value,out Result)"}.Result
    mirrorTransform.Translation=/Script/RigVM.RigVMFunction_MathVectorScale(Factor=-1, Value=@n2.Result.Location) {is=struct RigVMUnitNode, template="Scale::Execute(in Value,in Factor,out Result)"}.Result
    mirrorTransform.Scale3D=@n2.Result.Scale
   )
  ]}
 ContainedGraph {nodes [
   n1 = RigVMFunctionEntryNode()
   > RigVMVariableNode(Value=@n1.value_struct) {variable=value_struct_var}
   > RigVMVariableNode(Value=@n1.curve_struct) {variable=curve_struct_var}
   > SequenceExecution_2 = RigVMAggregateNode()
   RigVMFunction_MathQuaternionToEuler = &RigVMFunction_MathQuaternionToEuler(
    RotationOrder=ZYX
    Value=/Script/RigVM.RigVMFunction_MathQuaternionMakeRelative(
     Global=&RigUnit_GetTransform(Item=(Type=Bone,Name="retarget_calf_l"), Space=LocalSpace, bInitial=False).Transform.Rotation
     Parent=&RigUnit_GetTransform(Item=(Type=Bone,Name="retarget_calf_l"), Space=LocalSpace, bInitial=true).Transform.Rotation
    ) {is=struct RigVMUnitNode, template="Make Relative::Execute(in Global,in Parent,out Local)"}.Local
   )
   n2 = &RigVMFunction_MathFloatRemap(SourceMaximum=100, SourceMinimum=10, bClamp=true, Value=@RigVMFunction_MathQuaternionToEuler.Result.Z)
   n3 = RigVMVariableNode(Value=@n2.Result, ExecuteContext=@SequenceExecution_2.A) {variable=l_calf_alpha}
   n4 = RigVMVariableNode() {variable=curve_struct_var}
   RigVMFunction_ControlFlowBranch = &RigVMFunction_ControlFlowBranch(Condition=@n4.Value.deformRig_l_knee_alpha_82_#1, ExecuteContext=@n3.ExecuteContext)
   RigVMDispatchNode(BoneName=root, Name=deformRig_l_knee_alpha, ExecuteContext=@RigVMFunction_ControlFlowBranch.True, Value=@n2.Result) {template="DISPATCH_RigDispatch_SetAnimAttribute(in Name,in BoneName,in Value,out Success)"}
   n5 = &RigUnit_GetTransform(Item=(Type=Bone,Name="retarget_calf_r"), Space=LocalSpace, bInitial=true)
   n6 = &RigUnit_GetTransform(Item=(Type=Bone,Name="retarget_calf_r"), Space=LocalSpace, bInitial=False)
   n7 = &RigVMFunction_MathFloatRemap(
    SourceMaximum=100
    SourceMinimum=10
    bClamp=true
    Value=&RigVMFunction_MathQuaternionToEuler(
     RotationOrder=ZYX
     Value=&RigVMFunction_MathQuaternionMakeRelative(Parent=@n5.Transform.Rotation, Global=@n6.Transform.Rotation) {template="Make Relative::Execute(in Global,in Parent,out Local)"}.Local
    ).Result.Z
   )
   n8 = RigVMVariableNode(Value=@n7.Result, ExecuteContext=@SequenceExecution_2.B) {variable=r_calf_alpha}
   n9 = &RigVMFunction_ControlFlowBranch(Condition=@n4.Value.deformRig_r_knee_alpha_83_#2, ExecuteContext=@n8.ExecuteContext)
   RigVMDispatchNode(BoneName=root, Name=deformRig_r_knee_alpha, ExecuteContext=@n9.True, Value=@n7.Result) {template="DISPATCH_RigDispatch_SetAnimAttribute(in Name,in BoneName,in Value,out Success)"}
   SequenceExecution = RigVMAggregateNode(ExecuteContext=@SequenceExecution_2.C)
   SequenceExecution_1 = RigVMAggregateNode(ExecuteContext=@SequenceExecution.A)
   &RigUnit_ParentConstraint(
    AdvancedSettings=(RotationOrderForFilter=XZY)
    Child=(Type=Bone,Name="deform_knee_in_base_l")
    Parents=((Item=(Type=Bone,Name="retarget_calf_l"),Weight=0.5),(Item=(Type=Bone,Name="retarget_thigh_l"),Weight=0.5))
    Weight=1
    ExecutePin=@SequenceExecution_1.A
   )
   n10 = RigVMVariableNode() {variable=value_struct_var}
   > &RigUnit_ModifyTransforms(
    ItemToModify=((Item=(Type=Bone,Name="deform_knee_in_l"),Transform=(Rotation=(X=0,Y=0,Z=0,W=1),Translation=(X=0,Y=0,Z=0),Scale3D=(X=1,Y=1,Z=1))))
    ItemToModify.0.Transform=@n10.Value.knee_in_41_#3
    Weight=RigVMVariableNode() {variable=l_calf_alpha}.Value
   )
   &RigUnit_ParentConstraint(
    AdvancedSettings=(RotationOrderForFilter=XZY)
    Child=(Type=Bone,Name="deform_knee_in_base_r")
    Parents=((Item=(Type=Bone,Name="retarget_calf_r"),Weight=0.5),(Item=(Type=Bone,Name="retarget_thigh_r"),Weight=0.5))
    Weight=1
    ExecutePin=@SequenceExecution_1.B
   )
   > DeformRig_GetMirrorTransform_FNC = RigVMFunctionReferenceNode(transform=@n10.Value.knee_in_41_#3)
   > &RigUnit_ModifyTransforms(
    ItemToModify=((Item=(Type=Bone,Name="deform_knee_in_r"),Transform=(Rotation=(X=0,Y=0,Z=0,W=1),Translation=(X=0,Y=0,Z=0),Scale3D=(X=1,Y=1,Z=1))))
    ItemToModify.0.Transform=@DeformRig_GetMirrorTransform_FNC.mirrorTransform
    Weight=RigVMVariableNode() {variable=r_calf_alpha}.Value
   )
   SequenceExecution_1_1 = RigVMAggregateNode(ExecuteContext=@SequenceExecution.B)
   &RigUnit_ParentConstraint(
    AdvancedSettings=(RotationOrderForFilter=XZY)
    Child=(Type=Bone,Name="deform_knee_out_base_l")
    Parents=((Item=(Type=Bone,Name="retarget_calf_l"),Weight=0.1),(Item=(Type=Bone,Name="retarget_thigh_l"),Weight=0.9))
    Weight=1
    ExecutePin=@SequenceExecution_1_1.A
   )
   n11 = RigVMVariableNode() {variable=value_struct_var}
   > &RigUnit_ModifyTransforms(
    ItemToModify=((Item=(Type=Bone,Name="deform_knee_out_l"),Transform=(Rotation=(X=0,Y=0,Z=0,W=1),Translation=(X=0,Y=0,Z=0),Scale3D=(X=1,Y=1,Z=1))))
    ItemToModify.0.Transform=@n11.Value.knee_out_42_#4
    Weight=RigVMVariableNode() {variable=l_calf_alpha}.Value
   )
   &RigUnit_ParentConstraint(
    AdvancedSettings=(RotationOrderForFilter=XZY)
    Child=(Type=Bone,Name="deform_knee_out_base_r")
    Parents=((Item=(Type=Bone,Name="retarget_calf_r"),Weight=0.1),(Item=(Type=Bone,Name="retarget_thigh_r"),Weight=0.9))
    Weight=1
    ExecutePin=@SequenceExecution_1_1.B
   )
   > DeformRig_GetMirrorTransform_FNC_1 = RigVMFunctionReferenceNode(transform=@n11.Value.knee_out_42_#4)
   > &RigUnit_ModifyTransforms(
    ItemToModify=((Item=(Type=Bone,Name="deform_knee_out_r"),Transform=(Rotation=(X=0,Y=0,Z=0,W=1),Translation=(X=0,Y=0,Z=0),Scale3D=(X=1,Y=1,Z=1))))
    ItemToModify.0.Transform=@DeformRig_GetMirrorTransform_FNC_1.mirrorTransform
    Weight=RigVMVariableNode() {variable=r_calf_alpha}.Value
   )
   RigVMFunctionReturnNode(ExecuteContext=@SequenceExecution_2.D)
   DeformRig_GetDeltaRefTransform_FNC = RigVMFunctionReferenceNode(ExecuteContext=(), Item=(Type=Bone,Name="retarget_calf_l"))
   Get Delta Ref = RigVMFunctionReferenceNode(ExecuteContext=(), Item=(Type=Bone,Name="retarget_calf_l"))
  ]}
 ContainedGraph {nodes [
   n1 = RigVMFunctionEntryNode()
   SequenceExecution = &RigVMFunction_Sequence(ExecuteContext=@n1.ExecuteContext)
   SequenceExecution_1 = &RigVMFunction_Sequence(ExecuteContext=@SequenceExecution.B)
   SequenceExecution_2 = &RigVMFunction_Sequence(ExecuteContext=@SequenceExecution_1.B)
   RigVMFunctionReturnNode(A=@SequenceExecution.A, B=@SequenceExecution_1.A, C=@SequenceExecution_2.A, D=@SequenceExecution_2.B)
  ]}
 ContainedGraph {nodes [
   n1 = RigVMFunctionEntryNode()
   SequenceExecution_2 = &RigVMFunction_Sequence(ExecuteContext=@n1.ExecuteContext)
   SequenceExecution_2_1 = &RigVMFunction_Sequence(ExecuteContext=@SequenceExecution_2.B)
   SequenceExecution_2_2 = &RigVMFunction_Sequence(ExecuteContext=@SequenceExecution_2_1.B)
   RigVMFunctionReturnNode(A=@SequenceExecution_2.A, B=@SequenceExecution_2_1.A, C=@SequenceExecution_2_2.A, D=@SequenceExecution_2_2.B)
  ]}
 ContainedGraph {nodes [
   n1 = RigVMFunctionEntryNode()
   SequenceExecution_1_1 = &RigVMFunction_Sequence(ExecuteContext=@n1.ExecuteContext)
   SequenceExecution_1_1_1 = &RigVMFunction_Sequence(ExecuteContext=@SequenceExecution_1_1.B)
   SequenceExecution_1_1_2 = &RigVMFunction_Sequence(ExecuteContext=@SequenceExecution_1_1_1.B)
   RigVMFunctionReturnNode(A=@SequenceExecution_1_1.A, B=@SequenceExecution_1_1_1.A, C=@SequenceExecution_1_1_2.A, D=@SequenceExecution_1_1_2.B)
  ]}
 ContainedGraph {nodes [
   n1 = RigVMFunctionEntryNode()
   SequenceExecution_1 = &RigVMFunction_Sequence(ExecuteContext=@n1.ExecuteContext)
   SequenceExecution_1_1 = &RigVMFunction_Sequence(ExecuteContext=@SequenceExecution_1.B)
   SequenceExecution_1_2 = &RigVMFunction_Sequence(ExecuteContext=@SequenceExecution_1_1.B)
   RigVMFunctionReturnNode(A=@SequenceExecution_1.A, B=@SequenceExecution_1_1.A, C=@SequenceExecution_1_2.A, D=@SequenceExecution_1_2.B)
  ]}
 ContainedGraph {nodes [
   n1 = RigVMFunctionEntryNode()
   n2 = &RigVMFunction_MathTransformToEulerTransform(Value=@n1.transform) {template="Cast::Execute(in Value,out Result)"}
   > RigVMFunctionReturnNode(
    mirrorTransform=(Rotation=(X=0,Y=0,Z=0,W=1),Translation=(X=0,Y=0,Z=0),Scale3D=(X=0,Y=0,Z=0))
    mirrorTransform.Rotation=&RigVMFunction_MathQuaternionFromRotatorV2(Value=@n2.Result.Rotation) {template="Cast::Execute(in Value,out Result)"}.Result
    mirrorTransform.Translation=&RigVMFunction_MathVectorScale(Factor=-1, Value=@n2.Result.Location) {template="Scale::Execute(in Value,in Factor,out Result)"}.Result
    mirrorTransform.Scale3D=@n2.Result.Scale
   )
  ]}
 ContainedGraph {nodes [
   n1 = RigVMFunctionEntryNode()
   > RigVMVariableNode(Value=@n1.value_struct) {variable=value_struct_var}
   > RigVMVariableNode(Value=@n1.curve_struct) {variable=curve_struct_var}
   > SequenceExecution = RigVMAggregateNode()
   n2 = &RigUnit_SphericalPoseReader(
    ActiveRegionScaleFactors=(PositiveWidth=3.5,NegativeWidth=0.5,PositiveHeight=3,NegativeHeight=3)
    ActiveRegionSize=0.1
    Debug=(bDrawDebug=false,DebugScale=25,DebugSegments=20,DebugThickness=0.25)
    DriverAxis=(X=-1,Y=0,Z=0)
    DriverItem=(Type=Bone,Name="retarget_thigh_l")
    FalloffRegionScaleFactors=(PositiveWidth=1,NegativeWidth=1,PositiveHeight=1,NegativeHeight=1)
    FalloffSize=0.45
    OptionalParentItem=(Type=Bone)
    RotationOffset=(X=90,Y=0,Z=0)
    ExecutePin=@SequenceExecution.A
    Debug.bDrawDebug=@n1.debug
   )
   > n3 = &RigUnit_SphericalPoseReader(
    ActiveRegionScaleFactors=(PositiveWidth=3.5,NegativeWidth=0.5,PositiveHeight=3,NegativeHeight=3)
    ActiveRegionSize=0.1
    Debug=(bDrawDebug=false,DebugScale=25,DebugSegments=20,DebugThickness=0.25)
    DriverAxis=(X=1,Y=0,Z=0)
    DriverItem=(Type=Bone,Name="retarget_thigh_r")
    FalloffRegionScaleFactors=(PositiveWidth=1,NegativeWidth=1,PositiveHeight=1,NegativeHeight=1)
    FalloffSize=0.45
    FlipHeightScaling=false
    FlipWidthScaling=true
    OptionalParentItem=(Type=Bone)
    RotationOffset=(X=-90,Y=0,Z=0)
    Debug.bDrawDebug=@n1.debug
   )
   n4 = RigVMVariableNode() {variable=curve_struct_var}
   n5 = &RigVMFunction_ControlFlowBranch(ExecuteContext=@n3.ExecutePin, Condition=@n4.Value.deformRig_l_thigh_up_alpha_88_#5)
   n6 = RigVMDispatchNode(BoneName=root, Name=deformRig_l_thigh_up_alpha, ExecuteContext=@n5.True, Value=@n2.OutputParam) {template="DISPATCH_RigDispatch_SetAnimAttribute(in Name,in BoneName,in Value,out Success)"}
   n7 = &RigVMFunction_ControlFlowBranch(ExecuteContext=@n6.ExecuteContext, Condition=@n4.Value.deformRig_r_thigh_up_alpha_89_#6)
   RigVMDispatchNode(BoneName=root, Name=deformRig_r_thigh_up_alpha, ExecuteContext=@n7.True, Value=@n3.OutputParam) {template="DISPATCH_RigDispatch_SetAnimAttribute(in Name,in BoneName,in Value,out Success)"}
   RigVMFunctionReturnNode(ExecuteContext=@SequenceExecution.D)
   n8 = &RigVMFunction_ControlFlowBranch(ExecuteContext=@SequenceExecution.C, Condition=@n1.glute)
   SequenceExecution_1_1 = RigVMAggregateNode(ExecuteContext=@n8.True)
   &RigUnit_ParentConstraint(
    AdvancedSettings=(RotationOrderForFilter=XZY)
    Child=(Type=Bone,Name="deform_glute_base_l")
    Parents=((Item=(Type=Bone,Name="retarget_thigh_l"),Weight=0.25),(Item=(Type=Bone,Name="pelvis"),Weight=0.75))
    Weight=1
    ExecutePin=@SequenceExecution_1_1.A
   )
   n9 = RigVMVariableNode() {variable=value_struct_var}
   > &RigUnit_ModifyTransforms(
    ItemToModify=((Item=(Type=Bone,Name="deform_glute_l"),Transform=(Rotation=(X=0,Y=0,Z=0,W=1),Translation=(X=0,Y=0,Z=0),Scale3D=(X=1,Y=1,Z=1))))
    ItemToModify.0.Transform=@n9.Value.glute_36_#7
    Weight=@n2.OutputParam
   )
   &RigUnit_ParentConstraint(
    AdvancedSettings=(RotationOrderForFilter=XZY)
    Child=(Type=Bone,Name="deform_glute_base_r")
    Parents=((Item=(Type=Bone,Name="retarget_thigh_r"),Weight=0.25),(Item=(Type=Bone,Name="pelvis"),Weight=0.75))
    Weight=1
    ExecutePin=@SequenceExecution_1_1.B
   )
   > DeformRig_GetMirrorTransform_FNC_1 = RigVMFunctionReferenceNode(transform=@n9.Value.glute_36_#7)
   > &RigUnit_ModifyTransforms(
    ItemToModify=((Item=(Type=Bone,Name="deform_glute_r"),Transform=(Rotation=(X=0,Y=0,Z=0,W=1),Translation=(X=0,Y=0,Z=0),Scale3D=(X=1,Y=1,Z=1))))
    ItemToModify.0.Transform=@DeformRig_GetMirrorTransform_FNC_1.mirrorTransform
    Weight=@n3.OutputParam
   )
   RigVMFunction_ControlFlowBranch = &RigVMFunction_ControlFlowBranch(ExecuteContext=@SequenceExecution.B, Condition=@n1.groin)
   SequenceExecution_1 = RigVMAggregateNode(ExecuteContext=@RigVMFunction_ControlFlowBranch.True)
   &RigUnit_ParentConstraint(
    AdvancedSettings=(RotationOrderForFilter=XZY)
    Child=(Type=Bone,Name="deform_groin_base_l")
    Parents=((Item=(Type=Bone,Name="retarget_thigh_l"),Weight=0.7),(Item=(Type=Bone,Name="pelvis"),Weight=0.3))
    Weight=1
    ExecutePin=@SequenceExecution_1.A
   )
   n10 = RigVMVariableNode() {variable=value_struct_var}
   > &RigUnit_ModifyTransforms(
    ItemToModify=((Item=(Type=Bone,Name="deform_groin_l"),Transform=(Rotation=(X=0,Y=0,Z=0,W=1),Translation=(X=0,Y=0,Z=0),Scale3D=(X=1,Y=1,Z=1))))
    ItemToModify.0.Transform=@n10.Value.groin_35_#8
    Weight=@n2.OutputParam
   )
   &RigUnit_ParentConstraint(
    AdvancedSettings=(RotationOrderForFilter=XZY)
    Child=(Type=Bone,Name="deform_groin_base_r")
    Parents=((Item=(Type=Bone,Name="retarget_thigh_r"),Weight=0.7),(Item=(Type=Bone,Name="pelvis"),Weight=0.3))
    Weight=1
    ExecutePin=@SequenceExecution_1.B
   )
   > DeformRig_GetMirrorTransform_FNC = RigVMFunctionReferenceNode(transform=@n10.Value.groin_35_#8)
   > &RigUnit_ModifyTransforms(
    ItemToModify=((Item=(Type=Bone,Name="deform_groin_r"),Transform=(Rotation=(X=0,Y=0,Z=0,W=1),Translation=(X=0,Y=0,Z=0),Scale3D=(X=1,Y=1,Z=1))))
    ItemToModify.0.Transform=@DeformRig_GetMirrorTransform_FNC.mirrorTransform
    Weight=@n3.OutputParam
   )
  ]}
 ContainedGraph {nodes [
   n1 = RigVMFunctionEntryNode()
   SequenceExecution = &RigVMFunction_Sequence(ExecuteContext=@n1.ExecuteContext)
   SequenceExecution_1 = &RigVMFunction_Sequence(ExecuteContext=@SequenceExecution.B)
   SequenceExecution_2 = &RigVMFunction_Sequence(ExecuteContext=@SequenceExecution_1.B)
   RigVMFunctionReturnNode(A=@SequenceExecution.A, B=@SequenceExecution_1.A, C=@SequenceExecution_2.A, D=@SequenceExecution_2.B)
  ]}
 ContainedGraph {nodes [
   n1 = RigVMFunctionEntryNode()
   SequenceExecution_1 = &RigVMFunction_Sequence(ExecuteContext=@n1.ExecuteContext)
   SequenceExecution_1_1 = &RigVMFunction_Sequence(ExecuteContext=@SequenceExecution_1.B)
   SequenceExecution_1_2 = &RigVMFunction_Sequence(ExecuteContext=@SequenceExecution_1_1.B)
   RigVMFunctionReturnNode(A=@SequenceExecution_1.A, B=@SequenceExecution_1_1.A, C=@SequenceExecution_1_2.A, D=@SequenceExecution_1_2.B)
  ]}
 ContainedGraph {nodes [
   n1 = RigVMFunctionEntryNode()
   SequenceExecution_1_1 = &RigVMFunction_Sequence(ExecuteContext=@n1.ExecuteContext)
   SequenceExecution_1_1_1 = &RigVMFunction_Sequence(ExecuteContext=@SequenceExecution_1_1.B)
   SequenceExecution_1_1_2 = &RigVMFunction_Sequence(ExecuteContext=@SequenceExecution_1_1_1.B)
   RigVMFunctionReturnNode(A=@SequenceExecution_1_1.A, B=@SequenceExecution_1_1_1.A, C=@SequenceExecution_1_1_2.A, D=@SequenceExecution_1_1_2.B)
  ]}
 ContainedGraph {nodes [
   n1 = RigVMFunctionEntryNode()
   n2 = &RigUnit_GetTransform(Space=LocalSpace, Item=@n1.Item)
   n3 = &RigUnit_GetTransform(Space=LocalSpace, bInitial=true, Item=@n1.Item)
   > RigVMFunctionReturnNode(OutTransform=&RigVMFunction_MathTransformToEulerTransform(Value=&RigVMFunction_MathTransformMakeRelative(Global=@n2.Transform, Parent=@n3.Transform).Local) {template="Cast::Execute(in Value,out Result)"}.Result)
  ]}
 ContainedGraph {nodes [
   n1 = RigVMFunctionEntryNode()
   n2 = &RigVMFunction_MathTransformToEulerTransform(Value=@n1.transform) {template="Cast::Execute(in Value,out Result)"}
   > RigVMFunctionReturnNode(
    mirrorTransform=(Rotation=(X=0,Y=0,Z=0,W=1),Translation=(X=0,Y=0,Z=0),Scale3D=(X=0,Y=0,Z=0))
    mirrorTransform.Rotation=&RigVMFunction_MathQuaternionFromRotatorV2(Value=@n2.Result.Rotation) {template="Cast::Execute(in Value,out Result)"}.Result
    mirrorTransform.Translation=&RigVMFunction_MathVectorScale(Factor=-1, Value=@n2.Result.Location) {template="Scale::Execute(in Value,in Factor,out Result)"}.Result
    mirrorTransform.Scale3D=@n2.Result.Scale
   )
  ]}
 ContainedGraph {nodes [
   n1 = RigVMFunctionEntryNode()
   > RigVMVariableNode(Value=@n1.value_struct) {variable=value_struct_var}
   > RigVMVariableNode(Value=@n1.curve_struct) {variable=curve_struct_var}
   > SequenceExecution_2 = RigVMAggregateNode()
   DeformRig_GetDeltaRefTransform_FNC = RigVMFunctionReferenceNode(Item=(Type=Bone,Name="retarget_lowerarm_l"), ExecuteContext=@SequenceExecution_2.A)
   n2 = RigVMVariableNode() {variable=curve_struct_var}
   n3 = &RigVMFunction_ControlFlowBranch(
    ExecuteContext=@DeformRig_GetDeltaRefTransform_FNC.ExecuteContext
    Condition=@n2.Value.deformRig_l_elbow_alpha_86_#9
   )
   n4 = &RigVMFunction_MathFloatRemap(SourceMaximum=90, SourceMinimum=15, bClamp=true, Value=@DeformRig_GetDeltaRefTransform_FNC.OutTransform.Rotation.Yaw)
   RigVMDispatchNode(BoneName=root, Name=deformRig_l_elbow_alpha, Value=@n4.Result, ExecuteContext=@n3.True) {template="DISPATCH_RigDispatch_SetAnimAttribute(in Name,in BoneName,in Value,out Success)"}
   DeformRig_GetDeltaRefTransform_FNC_1 = RigVMFunctionReferenceNode(Item=(Type=Bone,Name="retarget_lowerarm_r"), ExecuteContext=@SequenceExecution_2.B)
   n5 = &RigVMFunction_ControlFlowBranch(
    ExecuteContext=@DeformRig_GetDeltaRefTransform_FNC_1.ExecuteContext
    Condition=@n2.Value.deformRig_r_elbow_alpha_87_#10
   )
   n6 = &RigVMFunction_MathFloatRemap(SourceMaximum=90, SourceMinimum=15, bClamp=true, Value=@DeformRig_GetDeltaRefTransform_FNC_1.OutTransform.Rotation.Yaw)
   RigVMDispatchNode(BoneName=root, Name=deformRig_r_elbow_alpha, Value=@n6.Result, ExecuteContext=@n5.True) {template="DISPATCH_RigDispatch_SetAnimAttribute(in Name,in BoneName,in Value,out Success)"}
   RigVMFunction_ControlFlowBranch = &RigVMFunction_ControlFlowBranch(ExecuteContext=@SequenceExecution_2.C, Condition=@n1.elbow)
   SequenceExecution = RigVMAggregateNode(ExecuteContext=@RigVMFunction_ControlFlowBranch.True)
   SequenceExecution_1 = RigVMAggregateNode(ExecuteContext=@SequenceExecution.A)
   n7 = RigVMVariableNode() {variable=value_struct_var}
   &RigUnit_ModifyTransforms(
    ItemToModify=((Item=(Type=Bone,Name="deform_elbow_in_l"),Transform=(Rotation=(X=0,Y=0,Z=0,W=1),Translation=(X=0,Y=0,Z=0),Scale3D=(X=1,Y=1,Z=1))))
    ExecutePin=@SequenceExecution_1.A
    ItemToModify.0.Transform=@n7.Value.elbow_in_24_#11
    Weight=@n4.Result
   )
   DeformRig_GetMirrorTransform_FNC_1 = RigVMFunctionReferenceNode(ExecuteContext=@SequenceExecution_1.B, transform=@n7.Value.elbow_in_24_#11)
   > &RigUnit_ModifyTransforms(
    ItemToModify=((Item=(Type=Bone,Name="deform_elbow_in_r"),Transform=(Rotation=(X=0,Y=0,Z=0,W=1),Translation=(X=0,Y=0,Z=0),Scale3D=(X=1,Y=1,Z=1))))
    ItemToModify.0.Transform=@DeformRig_GetMirrorTransform_FNC_1.mirrorTransform
    Weight=@n6.Result
   )
   SequenceExecution_1_1 = RigVMAggregateNode(ExecuteContext=@SequenceExecution.B)
   n8 = RigVMVariableNode() {variable=value_struct_var}
   &RigUnit_ModifyTransforms(
    ItemToModify=((Item=(Type=Bone,Name="deform_elbow_out_l"),Transform=(Rotation=(X=0,Y=0,Z=0,W=1),Translation=(X=0,Y=0,Z=0),Scale3D=(X=1,Y=1,Z=1))))
    ExecutePin=@SequenceExecution_1_1.A
    ItemToModify.0.Transform=@n8.Value.elbow_out_26_#12
    Weight=@n4.Result
   )
   DeformRig_GetMirrorTransform_FNC_2 = RigVMFunctionReferenceNode(ExecuteContext=@SequenceExecution_1_1.B, transform=@n8.Value.elbow_out_26_#12)
   > &RigUnit_ModifyTransforms(
    ItemToModify=((Item=(Type=Bone,Name="deform_elbow_out_r"),Transform=(Rotation=(X=0,Y=0,Z=0,W=1),Translation=(X=0,Y=0,Z=0),Scale3D=(X=1,Y=1,Z=1))))
    ItemToModify.0.Transform=@DeformRig_GetMirrorTransform_FNC_2.mirrorTransform
    Weight=@n6.Result
   )
   n9 = &RigVMFunction_ControlFlowBranch(ExecuteContext=@SequenceExecution_2.D, Condition=@n1.bicep)
   SequenceExecution_1_1_1 = RigVMAggregateNode(ExecuteContext=@n9.True)
   n10 = RigVMVariableNode() {variable=value_struct_var}
   &RigUnit_ModifyTransforms(
    ItemToModify=((Item=(Type=Bone,Name="deform_bicep_l"),Transform=(Rotation=(X=0,Y=0,Z=0,W=1),Translation=(X=0,Y=0,Z=0),Scale3D=(X=1,Y=1,Z=1))))
    ItemToModify.0.Transform=@n10.Value.bicep_44_#13
    ExecutePin=@SequenceExecution_1_1_1.A
    Weight=@n4.Result
   )
   DeformRig_GetMirrorTransform_FNC = RigVMFunctionReferenceNode(transform=@n10.Value.bicep_44_#13, ExecuteContext=@SequenceExecution_1_1_1.B)
   > &RigUnit_ModifyTransforms(
    ItemToModify=((Item=(Type=Bone,Name="deform_bicep_r"),Transform=(Rotation=(X=0,Y=0,Z=0,W=1),Translation=(X=0,Y=0,Z=0),Scale3D=(X=1,Y=1,Z=1))))
    ItemToModify.0.Transform=@DeformRig_GetMirrorTransform_FNC.mirrorTransform
    Weight=@n6.Result
   )
   RigVMFunctionReturnNode(ExecuteContext=@SequenceExecution_2.E)
   &RigUnit_ParentConstraint(
    AdvancedSettings=(InterpolationType=Shortest,RotationOrderForFilter=XZY)
    Child=(Type=Bone,Name="deform_elbow_in_base_l")
    ExecutePin=()
    Parents=((Item=(Type=Bone,Name="retarget_lowerarm_l"),Weight=0.4),(Item=(Type=Bone,Name="retarget_upperarm_l"),Weight=0.6))
    Weight=1
   )
   &RigUnit_ParentConstraint(
    AdvancedSettings=(RotationOrderForFilter=XZY)
    Child=(Type=Bone,Name="deform_elbow_in_base_r")
    ExecutePin=()
    Parents=((Item=(Type=Bone,Name="retarget_lowerarm_r"),Weight=0.4),(Item=(Type=Bone,Name="retarget_upperarm_r"),Weight=0.6))
    Weight=1
   )
   &RigUnit_ParentConstraint(
    AdvancedSettings=(RotationOrderForFilter=XZY)
    Child=(Type=Bone,Name="deform_elbow_out_base_r")
    ExecutePin=()
    Parents=((Item=(Type=Bone,Name="retarget_lowerarm_r"),Weight=0.3),(Item=(Type=Bone,Name="retarget_upperarm_l"),Weight=0.7))
    Weight=1
   )
   &RigUnit_ParentConstraint(
    AdvancedSettings=(RotationOrderForFilter=XZY)
    Child=(Type=Bone,Name="deform_elbow_out_base_l")
    ExecutePin=()
    Parents=((Item=(Type=Bone,Name="retarget_lowerarm_l"),Weight=0.3),(Item=(Type=Bone,Name="retarget_upperarm_l"),Weight=0.7))
    Weight=1
   )
  ]}
 ContainedGraph {nodes [
   n1 = RigVMFunctionEntryNode()
   SequenceExecution_2 = &RigVMFunction_Sequence(ExecuteContext=@n1.ExecuteContext)
   SequenceExecution_2_1 = &RigVMFunction_Sequence(ExecuteContext=@SequenceExecution_2.B)
   SequenceExecution_2_2 = &RigVMFunction_Sequence(ExecuteContext=@SequenceExecution_2_1.B)
   SequenceExecution_2_3 = &RigVMFunction_Sequence(ExecuteContext=@SequenceExecution_2_2.B)
   RigVMFunctionReturnNode(A=@SequenceExecution_2.A, B=@SequenceExecution_2_1.A, C=@SequenceExecution_2_2.A, D=@SequenceExecution_2_3.A, E=@SequenceExecution_2_3.B)
  ]}
 ContainedGraph {nodes [
   n1 = RigVMFunctionEntryNode()
   SequenceExecution = &RigVMFunction_Sequence(ExecuteContext=@n1.ExecuteContext)
   SequenceExecution_1 = &RigVMFunction_Sequence(ExecuteContext=@SequenceExecution.B)
   SequenceExecution_2 = &RigVMFunction_Sequence(ExecuteContext=@SequenceExecution_1.B)
   RigVMFunctionReturnNode(A=@SequenceExecution.A, B=@SequenceExecution_1.A, C=@SequenceExecution_2.A, D=@SequenceExecution_2.B)
  ]}
 ContainedGraph {nodes [
   n1 = RigVMFunctionEntryNode()
   SequenceExecution_1 = &RigVMFunction_Sequence(ExecuteContext=@n1.ExecuteContext)
   SequenceExecution_1_1 = &RigVMFunction_Sequence(ExecuteContext=@SequenceExecution_1.B)
   SequenceExecution_1_2 = &RigVMFunction_Sequence(ExecuteContext=@SequenceExecution_1_1.B)
   RigVMFunctionReturnNode(A=@SequenceExecution_1.A, B=@SequenceExecution_1_1.A, C=@SequenceExecution_1_2.A, D=@SequenceExecution_1_2.B)
  ]}
 ContainedGraph {nodes [
   n1 = RigVMFunctionEntryNode()
   SequenceExecution_1_1 = &RigVMFunction_Sequence(ExecuteContext=@n1.ExecuteContext)
   SequenceExecution_1_1_1 = &RigVMFunction_Sequence(ExecuteContext=@SequenceExecution_1_1.B)
   SequenceExecution_1_1_2 = &RigVMFunction_Sequence(ExecuteContext=@SequenceExecution_1_1_1.B)
   RigVMFunctionReturnNode(A=@SequenceExecution_1_1.A, B=@SequenceExecution_1_1_1.A, C=@SequenceExecution_1_1_2.A, D=@SequenceExecution_1_1_2.B)
  ]}
 ContainedGraph {nodes [
   n1 = RigVMFunctionEntryNode()
   SequenceExecution_1_1_1 = &RigVMFunction_Sequence(ExecuteContext=@n1.ExecuteContext)
   SequenceExecution_1_1_1_1 = &RigVMFunction_Sequence(ExecuteContext=@SequenceExecution_1_1_1.B)
   SequenceExecution_1_1_1_2 = &RigVMFunction_Sequence(ExecuteContext=@SequenceExecution_1_1_1_1.B)
   RigVMFunctionReturnNode(A=@SequenceExecution_1_1_1.A, B=@SequenceExecution_1_1_1_1.A, C=@SequenceExecution_1_1_1_2.A, D=@SequenceExecution_1_1_1_2.B)
  ]}
 ContainedGraph {nodes [
   n1 = RigVMFunctionEntryNode()
   > RigVMVariableNode(Value=@n1.value_struct) {variable=value_struct_var}
   > RigVMVariableNode(Value=@n1.curve_struct) {variable=curve_struct_var}
   > SequenceExecution_2 = RigVMAggregateNode()
   RigVMFunctionReturnNode(ExecuteContext=@SequenceExecution_2.D)
   SequenceExecution_3_1 = RigVMAggregateNode(ExecuteContext=@SequenceExecution_2.B)
   &RigUnit_ParentConstraint(
    AdvancedSettings=(RotationOrderForFilter=XZY)
    Child=(Type=Bone,Name="deform_upperarm_fr_base_l")
    Parents=((Item=(Type=Bone,Name="retarget_upperarm_twist_01_l"),Weight=0.3),(Item=(Type=Bone,Name="clavicle_l"),Weight=0.7))
    Weight=1
    ExecutePin=@SequenceExecution_3_1.A
   )
   > n2 = &RigUnit_SphericalPoseReader(
    ActiveRegionScaleFactors=(PositiveWidth=0.5,NegativeWidth=3,PositiveHeight=1.5,NegativeHeight=1.5)
    ActiveRegionSize=0.1
    Debug=(bDrawDebug=false,DebugScale=25,DebugSegments=20,DebugThickness=0.25)
    DriverAxis=(X=1,Y=0,Z=0)
    DriverItem=(Type=Bone,Name="retarget_upperarm_l")
    FalloffRegionScaleFactors=(PositiveWidth=1,NegativeWidth=1,PositiveHeight=1,NegativeHeight=1)
    FalloffSize=0.4
    OptionalParentItem=(Type=Bone)
    RotationOffset=(X=90,Y=0,Z=0)
    Debug.bDrawDebug=@n1.debug
   )
   n3 = RigVMVariableNode() {variable=value_struct_var}
   > &RigUnit_ModifyTransforms(
    ItemToModify=((Item=(Type=Bone,Name="deform_upperarm_fr_l"),Transform=(Rotation=(X=0,Y=0,Z=0,W=1),Translation=(X=0,Y=0,Z=0),Scale3D=(X=1,Y=1,Z=1))))
    ItemToModify.0.Transform=@n3.Value.upperarm_fwd_fr_19_#14
    Weight=@n2.OutputParam
   )
   n4 = RigVMVariableNode() {variable=value_struct_var}
   > n5 = &RigUnit_ModifyTransforms(
    ItemToModify=((Item=(Type=Bone,Name="deform_upperarm_bk_l"),Transform=(Rotation=(X=0,Y=0,Z=0,W=1),Translation=(X=0,Y=0,Z=0),Scale3D=(X=1,Y=1,Z=1))))
    ItemToModify.0.Transform=@n4.Value.upperarm_fwd_bk_20_#15
    Weight=@n2.OutputParam
   )
   n6 = RigVMVariableNode() {variable=curve_struct_var}
   n7 = &RigVMFunction_ControlFlowBranch(ExecuteContext=@n5.ExecutePin, Condition=@n6.Value.deformRig_l_upperarm_fr_alpha_74_#16)
   RigVMDispatchNode(BoneName=root, Name=deformRig_l_upperarm_fr_alpha, ExecuteContext=@n7.True, Value=@n2.OutputParam) {template="DISPATCH_RigDispatch_SetAnimAttribute(in Name,in BoneName,in Value,out Success)"}
   &RigUnit_ParentConstraint(
    AdvancedSettings=(RotationOrderForFilter=XZY)
    Child=(Type=Bone,Name="deform_upperarm_fr_base_r")
    Parents=((Item=(Type=Bone,Name="retarget_upperarm_twist_01_r"),Weight=0.3),(Item=(Type=Bone,Name="clavicle_r"),Weight=0.7))
    Weight=1
    ExecutePin=@SequenceExecution_3_1.B
   )
   > n8 = &RigUnit_SphericalPoseReader(
    ActiveRegionScaleFactors=(PositiveWidth=0.5,NegativeWidth=3,PositiveHeight=1.5,NegativeHeight=1.5)
    ActiveRegionSize=0.1
    Debug=(bDrawDebug=false,DebugScale=25,DebugSegments=20,DebugThickness=0.25)
    DriverAxis=(X=-1,Y=0,Z=0)
    DriverItem=(Type=Bone,Name="retarget_upperarm_r")
    FalloffRegionScaleFactors=(PositiveWidth=1,NegativeWidth=1,PositiveHeight=1,NegativeHeight=1)
    FalloffSize=0.4
    FlipWidthScaling=true
    OptionalParentItem=(Type=Bone)
    RotationOffset=(X=90,Y=180,Z=180)
    Debug.bDrawDebug=@n1.debug
   )
   > DeformRig_GetMirrorTransform_FNC_2 = RigVMFunctionReferenceNode(transform=@n3.Value.upperarm_fwd_fr_19_#14)
   > &RigUnit_ModifyTransforms(
    ItemToModify=((Item=(Type=Bone,Name="deform_upperarm_fr_r"),Transform=(Rotation=(X=0,Y=0,Z=0,W=1),Translation=(X=0,Y=0,Z=0),Scale3D=(X=1,Y=1,Z=1))))
    ItemToModify.0.Transform=@DeformRig_GetMirrorTransform_FNC_2.mirrorTransform
    Weight=@n8.OutputParam
   )
   > DeformRig_GetMirrorTransform_FNC_1 = RigVMFunctionReferenceNode(transform=@n4.Value.upperarm_fwd_bk_20_#15)
   > n9 = &RigUnit_ModifyTransforms(
    ItemToModify=((Item=(Type=Bone,Name="deform_upperarm_bk_r"),Transform=(Rotation=(X=0,Y=0,Z=0,W=1),Translation=(X=0,Y=0,Z=0),Scale3D=(X=1,Y=1,Z=1))))
    ItemToModify.0.Transform=@DeformRig_GetMirrorTransform_FNC_1.mirrorTransform
    Weight=@n8.OutputParam
   )
   n10 = &RigVMFunction_ControlFlowBranch(ExecuteContext=@n9.ExecutePin, Condition=@n6.Value.deformRig_r_upperarm_fr_alpha_75_#17)
   RigVMDispatchNode(BoneName=root, Name=deformRig_r_upperarm_fr_alpha, ExecuteContext=@n10.True, Value=@n8.OutputParam) {template="DISPATCH_RigDispatch_SetAnimAttribute(in Name,in BoneName,in Value,out Success)"}
   SequenceExecution_3_1_1 = RigVMAggregateNode(ExecuteContext=@SequenceExecution_2.C)
   &RigUnit_ParentConstraint(
    AdvancedSettings=(RotationOrderForFilter=XZY)
    Child=(Type=Bone,Name="deform_upperarm_bk_base_l")
    Parents=((Item=(Type=Bone,Name="retarget_upperarm_twist_01_l"),Weight=0.3),(Item=(Type=Bone,Name="clavicle_l"),Weight=0.7))
    Weight=1
    ExecutePin=@SequenceExecution_3_1_1.A
   )
   > n11 = &RigUnit_SphericalPoseReader(
    ActiveRegionScaleFactors=(PositiveWidth=3,NegativeWidth=1,PositiveHeight=1.5,NegativeHeight=1.5)
    ActiveRegionSize=0.1
    Debug=(bDrawDebug=false,DebugScale=25,DebugSegments=20,DebugThickness=0.25)
    DriverAxis=(X=1,Y=0,Z=0)
    DriverItem=(Type=Bone,Name="retarget_upperarm_l")
    FalloffRegionScaleFactors=(PositiveWidth=1,NegativeWidth=1,PositiveHeight=1,NegativeHeight=1)
    FalloffSize=0.4
    OptionalParentItem=(Type=Bone)
    RotationOffset=(X=90,Y=0,Z=180)
    Debug.bDrawDebug=@n1.debug
   )
   n12 = RigVMVariableNode() {variable=value_struct_var}
   > n13 = &RigUnit_ModifyTransforms(
    ItemToModify=((Item=(Type=Bone,Name="deform_upperarm_bk_l"),Transform=(Rotation=(X=0,Y=0,Z=0,W=1),Translation=(X=0,Y=0,Z=0),Scale3D=(X=1,Y=1,Z=1))))
    ItemToModify.0.Transform=@n12.Value.upperarm_bk_22_#18
    Weight=@n11.OutputParam
   )
   n14 = RigVMVariableNode() {variable=curve_struct_var}
   n15 = &RigVMFunction_ControlFlowBranch(ExecuteContext=@n13.ExecutePin, Condition=@n14.Value.deformRig_l_upperarm_bk_alpha_76_#19)
   RigVMDispatchNode(BoneName=root, Name=deformRig_l_upperarm_bk_alpha, ExecuteContext=@n15.True, Value=@n11.OutputParam) {template="DISPATCH_RigDispatch_SetAnimAttribute(in Name,in BoneName,in Value,out Success)"}
   &RigUnit_ParentConstraint(
    AdvancedSettings=(RotationOrderForFilter=XZY)
    Child=(Type=Bone,Name="deform_upperarm_bk_base_r")
    Parents=((Item=(Type=Bone,Name="retarget_upperarm_twist_01_r"),Weight=0.3),(Item=(Type=Bone,Name="clavicle_r"),Weight=0.7))
    Weight=1
    ExecutePin=@SequenceExecution_3_1_1.B
   )
   > n16 = &RigUnit_SphericalPoseReader(
    ActiveRegionScaleFactors=(PositiveWidth=3,NegativeWidth=1,PositiveHeight=1.5,NegativeHeight=1.5)
    ActiveRegionSize=0.1
    Debug=(bDrawDebug=false,DebugScale=25,DebugSegments=20,DebugThickness=0.25)
    DriverAxis=(X=-1,Y=0,Z=0)
    DriverItem=(Type=Bone,Name="retarget_upperarm_r")
    FalloffRegionScaleFactors=(PositiveWidth=1,NegativeWidth=1,PositiveHeight=1,NegativeHeight=1)
    FalloffSize=0.4
    FlipHeightScaling=true
    FlipWidthScaling=false
    OptionalParentItem=(Type=Bone)
    RotationOffset=(X=90,Y=0,Z=0)
    Debug.bDrawDebug=@n1.debug
   )
   > DeformRig_GetMirrorTransform_FNC_3 = RigVMFunctionReferenceNode(transform=@n12.Value.upperarm_bk_22_#18)
   > n17 = &RigUnit_ModifyTransforms(
    ItemToModify=((Item=(Type=Bone,Name="deform_upperarm_bk_r"),Transform=(Rotation=(X=0,Y=0,Z=0,W=1),Translation=(X=0,Y=0,Z=0),Scale3D=(X=1,Y=1,Z=1))))
    ItemToModify.0.Transform=@DeformRig_GetMirrorTransform_FNC_3.mirrorTransform
    Weight=@n16.OutputParam
   )
   n18 = &RigVMFunction_ControlFlowBranch(ExecuteContext=@n17.ExecutePin, Condition=@n14.Value.deformRig_r_upperarm_bk_alpha_77_#20)
   RigVMDispatchNode(BoneName=root, Name=deformRig_r_upperarm_bk_alpha, ExecuteContext=@n18.True, Value=@n16.OutputParam) {template="DISPATCH_RigDispatch_SetAnimAttribute(in Name,in BoneName,in Value,out Success)"}
   SequenceExecution_3 = RigVMAggregateNode(ExecuteContext=@SequenceExecution_2.A)
   &RigUnit_ParentConstraint(
    AdvancedSettings=(RotationOrderForFilter=XZY)
    Child=(Type=Bone,Name="deform_upperarm_up_base_l")
    Parents=((Item=(Type=Bone,Name="retarget_upperarm_twist_01_l"),Weight=0.5),(Item=(Type=Bone,Name="clavicle_l"),Weight=0.5))
    Weight=1
    ExecutePin=@SequenceExecution_3.A
   )
   > n19 = &RigUnit_SphericalPoseReader(
    ActiveRegionScaleFactors=(PositiveWidth=2.5,NegativeWidth=3,PositiveHeight=3,NegativeHeight=3)
    ActiveRegionSize=0.1
    Debug=(bDrawDebug=false,DebugScale=25,DebugSegments=20,DebugThickness=0.25)
    DriverAxis=(X=1,Y=0,Z=0)
    DriverItem=(Type=Bone,Name="retarget_upperarm_l")
    FalloffRegionScaleFactors=(PositiveWidth=1,NegativeWidth=1,PositiveHeight=1,NegativeHeight=1)
    FalloffSize=0.5
    OptionalParentItem=(Type=Bone)
    RotationOffset=(X=0,Y=0,Z=0)
    Debug.bDrawDebug=@n1.debug
   )
   n20 = RigVMVariableNode() {variable=value_struct_var}
   > n21 = &RigUnit_ModifyTransforms(
    ItemToModify=((Item=(Type=Bone,Name="deform_upperarm_up_l"),Transform=(Rotation=(X=0,Y=0,Z=0,W=1),Translation=(X=0,Y=0,Z=0),Scale3D=(X=1,Y=1,Z=1))))
    Weight=@n19.OutputParam
    ItemToModify.0.Transform=@n20.Value.upperarm_up_18_#21
   )
   n22 = RigVMVariableNode() {variable=curve_struct_var}
   RigVMFunction_ControlFlowBranch = &RigVMFunction_ControlFlowBranch(ExecuteContext=@n21.ExecutePin, Condition=@n22.Value.deformRig_l_upperarm_up_alpha_72_#22)
   RigVMDispatchNode(BoneName=root, Name=deformRig_l_upperarm_up_alpha, ExecuteContext=@RigVMFunction_ControlFlowBranch.True, Value=@n19.OutputParam) {template="DISPATCH_RigDispatch_SetAnimAttribute(in Name,in BoneName,in Value,out Success)"}
   &RigUnit_ParentConstraint(
    AdvancedSettings=(RotationOrderForFilter=XZY)
    Child=(Type=Bone,Name="deform_upperarm_up_base_r")
    Parents=((Item=(Type=Bone,Name="retarget_upperarm_twist_01_r"),Weight=0.5),(Item=(Type=Bone,Name="clavicle_r"),Weight=0.5))
    Weight=1
    ExecutePin=@SequenceExecution_3.B
   )
   > n23 = &RigUnit_SphericalPoseReader(
    ActiveRegionScaleFactors=(PositiveWidth=2.5,NegativeWidth=3,PositiveHeight=3,NegativeHeight=3)
    ActiveRegionSize=0.1
    Debug=(bDrawDebug=false,DebugScale=25,DebugSegments=20,DebugThickness=0.25)
    DriverAxis=(X=-1,Y=0,Z=0)
    DriverItem=(Type=Bone,Name="retarget_upperarm_r")
    FalloffRegionScaleFactors=(PositiveWidth=1,NegativeWidth=1,PositiveHeight=1,NegativeHeight=1)
    FalloffSize=0.5
    FlipHeightScaling=true
    OptionalParentItem=(Type=Bone)
    RotationOffset=(X=180,Y=0,Z=180)
    Debug.bDrawDebug=@n1.debug
   )
   > DeformRig_GetMirrorTransform_FNC = RigVMFunctionReferenceNode(transform=@n20.Value.upperarm_up_18_#21)
   > n24 = &RigUnit_ModifyTransforms(
    ItemToModify=((Item=(Type=Bone,Name="deform_upperarm_up_r"),Transform=(Rotation=(X=0,Y=0,Z=0,W=1),Translation=(X=0,Y=0,Z=0),Scale3D=(X=1,Y=1,Z=1))))
    ItemToModify.0.Transform=@DeformRig_GetMirrorTransform_FNC.mirrorTransform
    Weight=@n23.OutputParam
   )
   n25 = &RigVMFunction_ControlFlowBranch(ExecuteContext=@n24.ExecutePin, Condition=@n22.Value.deformRig_r_upperarm_up_alpha_73_#23)
   RigVMDispatchNode(BoneName=root, Name=deformRig_r_upperarm_up_alpha, ExecuteContext=@n25.True, Value=@n23.OutputParam) {template="DISPATCH_RigDispatch_SetAnimAttribute(in Name,in BoneName,in Value,out Success)"}
  ]}
 ContainedGraph {nodes [
   n1 = RigVMFunctionEntryNode()
   SequenceExecution_2 = &RigVMFunction_Sequence(ExecuteContext=@n1.ExecuteContext)
   SequenceExecution_2_1 = &RigVMFunction_Sequence(ExecuteContext=@SequenceExecution_2.B)
   SequenceExecution_2_2 = &RigVMFunction_Sequence(ExecuteContext=@SequenceExecution_2_1.B)
   RigVMFunctionReturnNode(A=@SequenceExecution_2.A, B=@SequenceExecution_2_1.A, C=@SequenceExecution_2_2.A, D=@SequenceExecution_2_2.B)
  ]}
 ContainedGraph {nodes [
   n1 = RigVMFunctionEntryNode()
   SequenceExecution_3_1 = &RigVMFunction_Sequence(ExecuteContext=@n1.ExecuteContext)
   SequenceExecution_3_1_1 = &RigVMFunction_Sequence(ExecuteContext=@SequenceExecution_3_1.B)
   SequenceExecution_3_1_2 = &RigVMFunction_Sequence(ExecuteContext=@SequenceExecution_3_1_1.B)
   RigVMFunctionReturnNode(A=@SequenceExecution_3_1.A, B=@SequenceExecution_3_1_1.A, C=@SequenceExecution_3_1_2.A, D=@SequenceExecution_3_1_2.B)
  ]}
 ContainedGraph {nodes [
   n1 = RigVMFunctionEntryNode()
   SequenceExecution_3 = &RigVMFunction_Sequence(ExecuteContext=@n1.ExecuteContext)
   SequenceExecution_3_1 = &RigVMFunction_Sequence(ExecuteContext=@SequenceExecution_3.B)
   SequenceExecution_3_2 = &RigVMFunction_Sequence(ExecuteContext=@SequenceExecution_3_1.B)
   RigVMFunctionReturnNode(A=@SequenceExecution_3.A, B=@SequenceExecution_3_1.A, C=@SequenceExecution_3_2.A, D=@SequenceExecution_3_2.B)
  ]}
 ContainedGraph {nodes [
   n1 = RigVMFunctionEntryNode()
   SequenceExecution_3_1_1 = &RigVMFunction_Sequence(ExecuteContext=@n1.ExecuteContext)
   SequenceExecution_3_1_1_1 = &RigVMFunction_Sequence(ExecuteContext=@SequenceExecution_3_1_1.B)
   SequenceExecution_3_1_1_2 = &RigVMFunction_Sequence(ExecuteContext=@SequenceExecution_3_1_1_1.B)
   RigVMFunctionReturnNode(A=@SequenceExecution_3_1_1.A, B=@SequenceExecution_3_1_1_1.A, C=@SequenceExecution_3_1_1_2.A, D=@SequenceExecution_3_1_1_2.B)
  ]}
 ContainedGraph {nodes [
   n1 = RigVMFunctionEntryNode()
   > RigVMVariableNode(Value=@n1.value_struct) {variable=value_struct_var}
   > RigVMVariableNode(Value=@n1.curve_struct) {variable=curve_struct_var}
   > SequenceExecution_2 = RigVMAggregateNode()
   DeformRig_GetDeltaRefTransform_FNC = RigVMFunctionReferenceNode(Item=(Type=Bone,Name="calf_l"), ExecuteContext=@SequenceExecution_2.A)
   n2 = RigVMVariableNode() {variable=curve_struct_var}
   RigVMFunction_ControlFlowBranch = &RigVMFunction_ControlFlowBranch(
    ExecuteContext=@DeformRig_GetDeltaRefTransform_FNC.ExecuteContext
    Condition=@n2.Value.deformRig_l_knee_alpha_82_#1
   )
   n3 = &RigVMFunction_MathFloatRemap(SourceMaximum=100, SourceMinimum=10, bClamp=true, Value=@DeformRig_GetDeltaRefTransform_FNC.OutTransform.Rotation.Yaw)
   RigVMDispatchNode(BoneName=root, Name=deformRig_l_knee_alpha, ExecuteContext=@RigVMFunction_ControlFlowBranch.True, Value=@n3.Result) {template="DISPATCH_RigDispatch_SetAnimAttribute(in Name,in BoneName,in Value,out Success)"}
   DeformRig_GetDeltaRefTransform_FNC_1 = RigVMFunctionReferenceNode(Item=(Type=Bone,Name="calf_r"), ExecuteContext=@SequenceExecution_2.B)
   n4 = &RigVMFunction_ControlFlowBranch(
    ExecuteContext=@DeformRig_GetDeltaRefTransform_FNC_1.ExecuteContext
    Condition=@n2.Value.deformRig_r_knee_alpha_83_#2
   )
   n5 = &RigVMFunction_MathFloatRemap(SourceMaximum=100, SourceMinimum=10, bClamp=true, Value=@DeformRig_GetDeltaRefTransform_FNC_1.OutTransform.Rotation.Yaw)
   RigVMDispatchNode(BoneName=root, Name=deformRig_r_knee_alpha, ExecuteContext=@n4.True, Value=@n5.Result) {template="DISPATCH_RigDispatch_SetAnimAttribute(in Name,in BoneName,in Value,out Success)"}
   SequenceExecution = RigVMAggregateNode(ExecuteContext=@SequenceExecution_2.C)
   SequenceExecution_1 = RigVMAggregateNode(ExecuteContext=@SequenceExecution.A)
   &RigUnit_ParentConstraint(
    AdvancedSettings=(RotationOrderForFilter=XZY)
    Child=(Type=Bone,Name="deform_knee_in_base_l")
    Parents=((Item=(Type=Bone,Name="calf_l"),Weight=0.5),(Item=(Type=Bone,Name="thigh_l"),Weight=0.5))
    Weight=1
    ExecutePin=@SequenceExecution_1.A
   )
   n6 = RigVMVariableNode() {variable=value_struct_var}
   > &RigUnit_ModifyTransforms(
    ItemToModify=((Item=(Type=Bone,Name="deform_knee_in_l"),Transform=(Rotation=(X=0,Y=0,Z=0,W=1),Translation=(X=0,Y=0,Z=0),Scale3D=(X=1,Y=1,Z=1))))
    ItemToModify.0.Transform=@n6.Value.knee_in_41_#3
    Weight=@n3.Result
   )
   &RigUnit_ParentConstraint(
    AdvancedSettings=(RotationOrderForFilter=XZY)
    Child=(Type=Bone,Name="deform_knee_in_base_r")
    Parents=((Item=(Type=Bone,Name="calf_r"),Weight=0.5),(Item=(Type=Bone,Name="thigh_r"),Weight=0.5))
    Weight=1
    ExecutePin=@SequenceExecution_1.B
   )
   > DeformRig_GetMirrorTransform_FNC = RigVMFunctionReferenceNode(transform=@n6.Value.knee_in_41_#3)
   > &RigUnit_ModifyTransforms(
    ItemToModify=((Item=(Type=Bone,Name="deform_knee_in_r"),Transform=(Rotation=(X=0,Y=0,Z=0,W=1),Translation=(X=0,Y=0,Z=0),Scale3D=(X=1,Y=1,Z=1))))
    ItemToModify.0.Transform=@DeformRig_GetMirrorTransform_FNC.mirrorTransform
    Weight=@n5.Result
   )
   SequenceExecution_1_1 = RigVMAggregateNode(ExecuteContext=@SequenceExecution.B)
   &RigUnit_ParentConstraint(
    AdvancedSettings=(RotationOrderForFilter=XZY)
    Child=(Type=Bone,Name="deform_knee_out_base_l")
    Parents=((Item=(Type=Bone,Name="calf_l"),Weight=0.1),(Item=(Type=Bone,Name="thigh_l"),Weight=0.9))
    Weight=1
    ExecutePin=@SequenceExecution_1_1.A
   )
   n7 = RigVMVariableNode() {variable=value_struct_var}
   > &RigUnit_ModifyTransforms(
    ItemToModify=((Item=(Type=Bone,Name="deform_knee_out_l"),Transform=(Rotation=(X=0,Y=0,Z=0,W=1),Translation=(X=0,Y=0,Z=0),Scale3D=(X=1,Y=1,Z=1))))
    ItemToModify.0.Transform=@n7.Value.knee_out_42_#4
    Weight=@n3.Result
   )
   &RigUnit_ParentConstraint(
    AdvancedSettings=(RotationOrderForFilter=XZY)
    Child=(Type=Bone,Name="deform_knee_out_base_r")
    Parents=((Item=(Type=Bone,Name="calf_r"),Weight=0.1),(Item=(Type=Bone,Name="thigh_r"),Weight=0.9))
    Weight=1
    ExecutePin=@SequenceExecution_1_1.B
   )
   > DeformRig_GetMirrorTransform_FNC_1 = RigVMFunctionReferenceNode(transform=@n7.Value.knee_out_42_#4)
   > &RigUnit_ModifyTransforms(
    ItemToModify=((Item=(Type=Bone,Name="deform_knee_out_r"),Transform=(Rotation=(X=0,Y=0,Z=0,W=1),Translation=(X=0,Y=0,Z=0),Scale3D=(X=1,Y=1,Z=1))))
    ItemToModify.0.Transform=@DeformRig_GetMirrorTransform_FNC_1.mirrorTransform
    Weight=@n5.Result
   )
   RigVMFunctionReturnNode(ExecuteContext=@SequenceExecution_2.D)
  ]}
 ContainedGraph {nodes [
   n1 = RigVMFunctionEntryNode()
   SequenceExecution = &RigVMFunction_Sequence(ExecuteContext=@n1.ExecuteContext)
   SequenceExecution_1 = &RigVMFunction_Sequence(ExecuteContext=@SequenceExecution.B)
   SequenceExecution_2 = &RigVMFunction_Sequence(ExecuteContext=@SequenceExecution_1.B)
   RigVMFunctionReturnNode(A=@SequenceExecution.A, B=@SequenceExecution_1.A, C=@SequenceExecution_2.A, D=@SequenceExecution_2.B)
  ]}
 ContainedGraph {nodes [
   n1 = RigVMFunctionEntryNode()
   SequenceExecution_2 = &RigVMFunction_Sequence(ExecuteContext=@n1.ExecuteContext)
   SequenceExecution_2_1 = &RigVMFunction_Sequence(ExecuteContext=@SequenceExecution_2.B)
   SequenceExecution_2_2 = &RigVMFunction_Sequence(ExecuteContext=@SequenceExecution_2_1.B)
   RigVMFunctionReturnNode(A=@SequenceExecution_2.A, B=@SequenceExecution_2_1.A, C=@SequenceExecution_2_2.A, D=@SequenceExecution_2_2.B)
  ]}
 ContainedGraph {nodes [
   n1 = RigVMFunctionEntryNode()
   SequenceExecution_1_1 = &RigVMFunction_Sequence(ExecuteContext=@n1.ExecuteContext)
   SequenceExecution_1_1_1 = &RigVMFunction_Sequence(ExecuteContext=@SequenceExecution_1_1.B)
   SequenceExecution_1_1_2 = &RigVMFunction_Sequence(ExecuteContext=@SequenceExecution_1_1_1.B)
   RigVMFunctionReturnNode(A=@SequenceExecution_1_1.A, B=@SequenceExecution_1_1_1.A, C=@SequenceExecution_1_1_2.A, D=@SequenceExecution_1_1_2.B)
  ]}
 ContainedGraph {nodes [
   n1 = RigVMFunctionEntryNode()
   SequenceExecution_1 = &RigVMFunction_Sequence(ExecuteContext=@n1.ExecuteContext)
   SequenceExecution_1_1 = &RigVMFunction_Sequence(ExecuteContext=@SequenceExecution_1.B)
   SequenceExecution_1_2 = &RigVMFunction_Sequence(ExecuteContext=@SequenceExecution_1_1.B)
   RigVMFunctionReturnNode(A=@SequenceExecution_1.A, B=@SequenceExecution_1_1.A, C=@SequenceExecution_1_2.A, D=@SequenceExecution_1_2.B)
  ]}
 ContainedGraph {nodes [
   n1 = RigVMFunctionEntryNode()
   > RigVMVariableNode(Value=@n1.value_struct) {variable=value_struct_var}
   > RigVMVariableNode(Value=@n1.curve_struct) {variable=curve_struct_var}
   > SequenceExecution = RigVMAggregateNode()
   n2 = &RigUnit_SphericalPoseReader(
    ActiveRegionScaleFactors=(PositiveWidth=3.5,NegativeWidth=0.5,PositiveHeight=3,NegativeHeight=3)
    ActiveRegionSize=0.1
    Debug=(bDrawDebug=false,DebugScale=25,DebugSegments=20,DebugThickness=0.25)
    DriverAxis=(X=-1,Y=0,Z=0)
    DriverItem=(Type=Bone,Name="thigh_l")
    FalloffRegionScaleFactors=(PositiveWidth=1,NegativeWidth=1,PositiveHeight=1,NegativeHeight=1)
    FalloffSize=0.45
    OptionalParentItem=(Type=Bone)
    RotationOffset=(X=90,Y=0,Z=0)
    ExecutePin=@SequenceExecution.A
    Debug.bDrawDebug=@n1.debug
   )
   > n3 = &RigUnit_SphericalPoseReader(
    ActiveRegionScaleFactors=(PositiveWidth=3.5,NegativeWidth=0.5,PositiveHeight=3,NegativeHeight=3)
    ActiveRegionSize=0.1
    Debug=(bDrawDebug=false,DebugScale=25,DebugSegments=20,DebugThickness=0.25)
    DriverAxis=(X=1,Y=0,Z=0)
    DriverItem=(Type=Bone,Name="thigh_r")
    FalloffRegionScaleFactors=(PositiveWidth=1,NegativeWidth=1,PositiveHeight=1,NegativeHeight=1)
    FalloffSize=0.45
    FlipHeightScaling=false
    FlipWidthScaling=true
    OptionalParentItem=(Type=Bone)
    RotationOffset=(X=-90,Y=0,Z=0)
    Debug.bDrawDebug=@n1.debug
   )
   n4 = RigVMVariableNode() {variable=curve_struct_var}
   n5 = &RigVMFunction_ControlFlowBranch(ExecuteContext=@n3.ExecutePin, Condition=@n4.Value.deformRig_l_thigh_up_alpha_88_#5)
   n6 = RigVMDispatchNode(BoneName=root, Name=deformRig_l_thigh_up_alpha, ExecuteContext=@n5.True, Value=@n2.OutputParam) {template="DISPATCH_RigDispatch_SetAnimAttribute(in Name,in BoneName,in Value,out Success)"}
   n7 = &RigVMFunction_ControlFlowBranch(ExecuteContext=@n6.ExecuteContext, Condition=@n4.Value.deformRig_r_thigh_up_alpha_89_#6)
   RigVMDispatchNode(BoneName=root, Name=deformRig_r_thigh_up_alpha, ExecuteContext=@n7.True, Value=@n3.OutputParam) {template="DISPATCH_RigDispatch_SetAnimAttribute(in Name,in BoneName,in Value,out Success)"}
   RigVMFunctionReturnNode(ExecuteContext=@SequenceExecution.D)
   n8 = &RigVMFunction_ControlFlowBranch(ExecuteContext=@SequenceExecution.C, Condition=@n1.glute)
   SequenceExecution_1_1 = RigVMAggregateNode(ExecuteContext=@n8.True)
   &RigUnit_ParentConstraint(
    AdvancedSettings=(RotationOrderForFilter=XZY)
    Child=(Type=Bone,Name="deform_glute_base_l")
    Parents=((Item=(Type=Bone,Name="thigh_l"),Weight=0.25),(Item=(Type=Bone,Name="pelvis"),Weight=0.75))
    Weight=1
    ExecutePin=@SequenceExecution_1_1.A
   )
   n9 = RigVMVariableNode() {variable=value_struct_var}
   > &RigUnit_ModifyTransforms(
    ItemToModify=((Item=(Type=Bone,Name="deform_glute_l"),Transform=(Rotation=(X=0,Y=0,Z=0,W=1),Translation=(X=0,Y=0,Z=0),Scale3D=(X=1,Y=1,Z=1))))
    ItemToModify.0.Transform=@n9.Value.glute_36_#7
    Weight=@n2.OutputParam
   )
   &RigUnit_ParentConstraint(
    AdvancedSettings=(RotationOrderForFilter=XZY)
    Child=(Type=Bone,Name="deform_glute_base_r")
    Parents=((Item=(Type=Bone,Name="thigh_r"),Weight=0.25),(Item=(Type=Bone,Name="pelvis"),Weight=0.75))
    Weight=1
    ExecutePin=@SequenceExecution_1_1.B
   )
   > DeformRig_GetMirrorTransform_FNC_1 = RigVMFunctionReferenceNode(transform=@n9.Value.glute_36_#7)
   > &RigUnit_ModifyTransforms(
    ItemToModify=((Item=(Type=Bone,Name="deform_glute_r"),Transform=(Rotation=(X=0,Y=0,Z=0,W=1),Translation=(X=0,Y=0,Z=0),Scale3D=(X=1,Y=1,Z=1))))
    ItemToModify.0.Transform=@DeformRig_GetMirrorTransform_FNC_1.mirrorTransform
    Weight=@n3.OutputParam
   )
   RigVMFunction_ControlFlowBranch = &RigVMFunction_ControlFlowBranch(ExecuteContext=@SequenceExecution.B, Condition=@n1.groin)
   SequenceExecution_1 = RigVMAggregateNode(ExecuteContext=@RigVMFunction_ControlFlowBranch.True)
   &RigUnit_ParentConstraint(
    AdvancedSettings=(RotationOrderForFilter=XZY)
    Child=(Type=Bone,Name="deform_groin_base_l")
    Parents=((Item=(Type=Bone,Name="thigh_l"),Weight=0.7),(Item=(Type=Bone,Name="pelvis"),Weight=0.3))
    Weight=1
    ExecutePin=@SequenceExecution_1.A
   )
   n10 = RigVMVariableNode() {variable=value_struct_var}
   > &RigUnit_ModifyTransforms(
    ItemToModify=((Item=(Type=Bone,Name="deform_groin_l"),Transform=(Rotation=(X=0,Y=0,Z=0,W=1),Translation=(X=0,Y=0,Z=0),Scale3D=(X=1,Y=1,Z=1))))
    ItemToModify.0.Transform=@n10.Value.groin_35_#8
    Weight=@n2.OutputParam
   )
   &RigUnit_ParentConstraint(
    AdvancedSettings=(RotationOrderForFilter=XZY)
    Child=(Type=Bone,Name="deform_groin_base_r")
    Parents=((Item=(Type=Bone,Name="thigh_r"),Weight=0.7),(Item=(Type=Bone,Name="pelvis"),Weight=0.3))
    Weight=1
    ExecutePin=@SequenceExecution_1.B
   )
   > DeformRig_GetMirrorTransform_FNC = RigVMFunctionReferenceNode(transform=@n10.Value.groin_35_#8)
   > &RigUnit_ModifyTransforms(
    ItemToModify=((Item=(Type=Bone,Name="deform_groin_r"),Transform=(Rotation=(X=0,Y=0,Z=0,W=1),Translation=(X=0,Y=0,Z=0),Scale3D=(X=1,Y=1,Z=1))))
    ItemToModify.0.Transform=@DeformRig_GetMirrorTransform_FNC.mirrorTransform
    Weight=@n3.OutputParam
   )
  ]}
 ContainedGraph {nodes [
   n1 = RigVMFunctionEntryNode()
   SequenceExecution = &RigVMFunction_Sequence(ExecuteContext=@n1.ExecuteContext)
   SequenceExecution_1 = &RigVMFunction_Sequence(ExecuteContext=@SequenceExecution.B)
   SequenceExecution_2 = &RigVMFunction_Sequence(ExecuteContext=@SequenceExecution_1.B)
   RigVMFunctionReturnNode(A=@SequenceExecution.A, B=@SequenceExecution_1.A, C=@SequenceExecution_2.A, D=@SequenceExecution_2.B)
  ]}
 ContainedGraph {nodes [
   n1 = RigVMFunctionEntryNode()
   SequenceExecution_1 = &RigVMFunction_Sequence(ExecuteContext=@n1.ExecuteContext)
   SequenceExecution_1_1 = &RigVMFunction_Sequence(ExecuteContext=@SequenceExecution_1.B)
   SequenceExecution_1_2 = &RigVMFunction_Sequence(ExecuteContext=@SequenceExecution_1_1.B)
   RigVMFunctionReturnNode(A=@SequenceExecution_1.A, B=@SequenceExecution_1_1.A, C=@SequenceExecution_1_2.A, D=@SequenceExecution_1_2.B)
  ]}
 ContainedGraph {nodes [
   n1 = RigVMFunctionEntryNode()
   SequenceExecution_1_1 = &RigVMFunction_Sequence(ExecuteContext=@n1.ExecuteContext)
   SequenceExecution_1_1_1 = &RigVMFunction_Sequence(ExecuteContext=@SequenceExecution_1_1.B)
   SequenceExecution_1_1_2 = &RigVMFunction_Sequence(ExecuteContext=@SequenceExecution_1_1_1.B)
   RigVMFunctionReturnNode(A=@SequenceExecution_1_1.A, B=@SequenceExecution_1_1_1.A, C=@SequenceExecution_1_1_2.A, D=@SequenceExecution_1_1_2.B)
  ]}
 ContainedGraph {nodes [
   n1 = RigVMFunctionEntryNode()
   > RigVMVariableNode(Value=@n1.value_struct) {variable=value_struct_var}
   > RigVMVariableNode(Value=@n1.curve_struct) {variable=curve_struct_var}
   > SequenceExecution = RigVMAggregateNode()
   RigVMFunctionReturnNode(ExecuteContext=@SequenceExecution.D)
   n2 = /Script/ControlRig.RigUnit_AimBoneMath(
    DebugSettings=(Scale=10,WorldOffset=(Rotation=(X=0,Y=0,Z=0,W=1),Translation=(X=0,Y=0,Z=0),Scale3D=(X=1,Y=1,Z=1)))
    Primary=(Weight=1,Axis=(X=-1,Y=0,Z=0),Target=(X=0,Y=0,Z=0),Kind=Location,Space=(Type=Bone,Name="retarget_upperarm_r"))
    Secondary=(Weight=1,Axis=(X=0,Y=1,Z=0),Target=(X=0,Y=0,Z=0),Kind=Location,Space=(Type=Bone,Name="spine_04"))
    Weight=1
    InputTransform=&RigUnit_GetTransform(Item=(Type=Bone,Name="deform_pec_aim_r")).Transform
   ) {is=struct RigVMUnitNode}
   &RigUnit_SetTransform(Item=(Type=Bone,Name="deform_pec_aim_r"), Weight=1, bInitial=false, Value=@n2.Result, ExecutePin=@SequenceExecution.B)
   > /Script/ControlRig.RigUnit_PositionConstraintLocalSpaceOffset(
    Child=(Type=Bone,Name="deform_pec_r")
    Filter=(bX=True,bY=false,bZ=false)
    Parents=((Item=(Type=Bone,Name="deform_pec_base_r"),Weight=0),(Item=(Type=Bone,Name="retarget_upperarm_r"),Weight=1))
    Weight=1
    bMaintainOffset=True
   ) {is=struct RigVMUnitNode}
   > DeformRig_GetDeltaRefTransform_FNC_1 = RigVMFunctionReferenceNode(Item=(Type=Bone,Name="clavicle_r"))
   n3 = RigVMVariableNode() {variable=value_struct_var}
   > DeformRig_GetMirrorTransform_FNC = RigVMFunctionReferenceNode(transform=@n3.Value.pec_51_#24)
   n4 = &RigVMFunction_MathFloatRemap(SourceMaximum=30, bClamp=true, Value=@DeformRig_GetDeltaRefTransform_FNC_1.OutTransform.Rotation.Pitch)
   > n5 = &RigUnit_ModifyTransforms(
    ItemToModify=((Item=(Type=Bone,Name="deform_pec_r"),Transform=(Rotation=(X=0,Y=0,Z=0,W=1),Translation=(X=0,Y=0,Z=0),Scale3D=(X=1,Y=1,Z=1))))
    ItemToModify.0.Transform=@DeformRig_GetMirrorTransform_FNC.mirrorTransform
    Weight=@n4.Result
   )
   n6 = RigVMVariableNode() {variable=curve_struct_var}
   n7 = &RigVMFunction_ControlFlowBranch(ExecuteContext=@n5.ExecutePin, Condition=@n6.Value.deformRig_r_pec_alpha_85_#25)
   RigVMDispatchNode(BoneName=root, Name=deformRig_r_pec_alpha, Value=@n4.Result, ExecuteContext=@n7.True) {template="DISPATCH_RigDispatch_SetAnimAttribute(in Name,in BoneName,in Value,out Success)"}
   n8 = &RigUnit_AimBoneMath(
    DebugSettings=(Scale=10,WorldOffset=(Rotation=(X=0,Y=0,Z=0,W=1),Translation=(X=0,Y=0,Z=0),Scale3D=(X=1,Y=1,Z=1)))
    Primary=(Weight=1,Axis=(X=1,Y=0,Z=0),Target=(X=0,Y=0,Z=0),Kind=Location,Space=(Type=Bone,Name="retarget_upperarm_l"))
    Secondary=(Weight=1,Axis=(X=0,Y=-1,Z=0),Target=(X=0,Y=0,Z=0),Kind=Location,Space=(Type=Bone,Name="spine_04"))
    Weight=1
    InputTransform=&RigUnit_GetTransform(Item=(Type=Bone,Name="deform_pec_aim_l")).Transform
   )
   &RigUnit_SetTransform(Item=(Type=Bone,Name="deform_pec_aim_l"), Weight=1, bInitial=false, Value=@n8.Result, ExecutePin=@SequenceExecution.A)
   > &RigUnit_PositionConstraintLocalSpaceOffset(
    Child=(Type=Bone,Name="deform_pec_l")
    Filter=(bX=True,bY=false,bZ=false)
    Parents=((Item=(Type=Bone,Name="deform_pec_base_l"),Weight=0),(Item=(Type=Bone,Name="retarget_upperarm_twist_01_l"),Weight=1))
    Weight=1
    bMaintainOffset=True
   )
   > DeformRig_GetDeltaRefTransform_FNC = RigVMFunctionReferenceNode(Item=(Type=Bone,Name="clavicle_l"))
   n9 = &RigVMFunction_MathFloatRemap(SourceMaximum=30, bClamp=true, Value=@DeformRig_GetDeltaRefTransform_FNC.OutTransform.Rotation.Pitch)
   > n10 = &RigUnit_ModifyTransforms(
    ItemToModify=((Item=(Type=Bone,Name="deform_pec_l"),Transform=(Rotation=(X=0,Y=0,Z=0,W=1),Translation=(X=0,Y=0,Z=0),Scale3D=(X=1,Y=1,Z=1))))
    Weight=@n9.Result
    ItemToModify.0.Transform=@n3.Value.pec_51_#24
   )
   RigVMFunction_ControlFlowBranch = &RigVMFunction_ControlFlowBranch(ExecuteContext=@n10.ExecutePin, Condition=@n6.Value.deformRig_l_pec_alpha_84_#26)
   DISPATCH_RigDispatch_SetAnimAttribute = RigVMDispatchNode(BoneName=root, Name=deformRig_l_pec_alpha, Value=@n9.Result, ExecuteContext=@RigVMFunction_ControlFlowBranch.True) {template="DISPATCH_RigDispatch_SetAnimAttribute(in Name,in BoneName,in Value,out Success)"}
   &RigUnit_ParentConstraint(
    AdvancedSettings=(RotationOrderForFilter=XZY)
    Child=(Type=Bone,Name="deform_pec_aim_base_l")
    ExecutePin=()
    Parents=((Item=(Type=Bone,Name="spine_03"),Weight=0.25),(Item=(Type=Bone,Name="spine_04"),Weight=0.55),(Item=(Type=Bone,Name="spine_05"),Weight=0.2))
    Weight=1
   )
   &RigUnit_ParentConstraint(
    AdvancedSettings=(RotationOrderForFilter=XZY)
    Child=(Type=Bone,Name="deform_pec_aim_base_r")
    ExecutePin=()
    Parents=((Item=(Type=Bone,Name="spine_03"),Weight=0.25),(Item=(Type=Bone,Name="spine_04"),Weight=0.55),(Item=(Type=Bone,Name="spine_05"),Weight=0.2))
    Weight=1
   )
  ]}
 ContainedGraph {nodes [
   n1 = RigVMFunctionEntryNode()
   SequenceExecution = &RigVMFunction_Sequence(ExecuteContext=@n1.ExecuteContext)
   SequenceExecution_1 = &RigVMFunction_Sequence(ExecuteContext=@SequenceExecution.B)
   SequenceExecution_2 = &RigVMFunction_Sequence(ExecuteContext=@SequenceExecution_1.B)
   RigVMFunctionReturnNode(A=@SequenceExecution.A, B=@SequenceExecution_1.A, C=@SequenceExecution_2.A, D=@SequenceExecution_2.B)
  ]}
 ContainedGraph {nodes [
   n1 = RigVMFunctionEntryNode()
   > RigVMVariableNode(Value=@n1.value_struct) {variable=value_struct_var}
   > RigVMVariableNode(Value=@n1.curve_struct) {variable=curve_struct_var}
   > SequenceExecution = RigVMAggregateNode()
   RigVMFunctionReturnNode(ExecuteContext=@SequenceExecution.D)
   SequenceExecution_1 = RigVMAggregateNode(ExecuteContext=@SequenceExecution.A)
   &RigUnit_ParentConstraint(
    AdvancedSettings=(RotationOrderForFilter=XZY)
    Child=(Type=Bone,Name="deform_wrist_up_base_l")
    Parents=((Item=(Type=Bone,Name="retarget_hand_l"),Weight=0.4),(Item=(Type=Bone,Name="retarget_lowerarm_twist_01_l"),Weight=0.6))
    Weight=1
    ExecutePin=@SequenceExecution_1.A
   )
   > n2 = &RigUnit_SphericalPoseReader(
    ActiveRegionScaleFactors=(PositiveWidth=0.5,NegativeWidth=3.5,PositiveHeight=3,NegativeHeight=3)
    ActiveRegionSize=0.1
    Debug=(bDrawDebug=false,DebugScale=25,DebugSegments=20,DebugThickness=0.25)
    DriverAxis=(X=1,Y=0,Z=0)
    DriverItem=(Type=Bone,Name="retarget_hand_l")
    FalloffRegionScaleFactors=(PositiveWidth=1,NegativeWidth=1,PositiveHeight=1,NegativeHeight=1)
    FalloffSize=0.45
    OptionalParentItem=(Type=Bone,Name="retarget_lowerarm_twist_01_l")
    RotationOffset=(X=90,Y=0,Z=0)
    Debug.bDrawDebug=@n1.debug
   )
   n3 = RigVMVariableNode() {variable=value_struct_var}
   > n4 = &RigUnit_ModifyTransforms(
    ItemToModify=((Item=(Type=Bone,Name="deform_wrist_up_l"),Transform=(Rotation=(X=0,Y=0,Z=0,W=1),Translation=(X=0,Y=0,Z=0),Scale3D=(X=1,Y=1,Z=1))))
    ItemToModify.0.Transform=@n3.Value.wrist_up_31_#27
    Weight=@n2.OutputParam
   )
   n5 = RigVMVariableNode() {variable=curve_struct_var}
   RigVMFunction_ControlFlowBranch = &RigVMFunction_ControlFlowBranch(ExecuteContext=@n4.ExecutePin, Condition=@n5.Value.deformRig_l_wrist_up_alpha_78_#28)
   RigVMDispatchNode(BoneName=root, Name=deformRig_l_wrist_up_alpha, ExecuteContext=@RigVMFunction_ControlFlowBranch.True, Value=@n2.OutputParam) {template="DISPATCH_RigDispatch_SetAnimAttribute(in Name,in BoneName,in Value,out Success)"}
   &RigUnit_ParentConstraint(
    AdvancedSettings=(RotationOrderForFilter=XZY)
    Child=(Type=Bone,Name="deform_wrist_up_base_r")
    Parents=((Item=(Type=Bone,Name="retarget_hand_r"),Weight=0.4),(Item=(Type=Bone,Name="retarget_lowerarm_twist_01_r"),Weight=0.6))
    Weight=1
    ExecutePin=@SequenceExecution_1.B
   )
   > n6 = &RigUnit_SphericalPoseReader(
    ActiveRegionScaleFactors=(PositiveWidth=0.5,NegativeWidth=3.5,PositiveHeight=3,NegativeHeight=3)
    ActiveRegionSize=0.1
    Debug=(bDrawDebug=false,DebugScale=25,DebugSegments=20,DebugThickness=0.25)
    DriverAxis=(X=-1,Y=0,Z=0)
    DriverItem=(Type=Bone,Name="retarget_hand_r")
    FalloffRegionScaleFactors=(PositiveWidth=1,NegativeWidth=1,PositiveHeight=1,NegativeHeight=1)
    FalloffSize=0.45
    FlipHeightScaling=true
    OptionalParentItem=(Type=Bone,Name="retarget_lowerarm_twist_01_r")
    RotationOffset=(X=90,Y=0,Z=180)
    Debug.bDrawDebug=@n1.debug
   )
   > DeformRig_GetMirrorTransform_FNC = RigVMFunctionReferenceNode(transform=@n3.Value.wrist_up_31_#27)
   > n7 = &RigUnit_ModifyTransforms(
    ItemToModify=((Item=(Type=Bone,Name="deform_wrist_up_r"),Transform=(Rotation=(X=0,Y=0,Z=0,W=1),Translation=(X=0,Y=0,Z=0),Scale3D=(X=1,Y=1,Z=1))))
    ItemToModify.0.Transform=@DeformRig_GetMirrorTransform_FNC.mirrorTransform
    Weight=@n6.OutputParam
   )
   n8 = &RigVMFunction_ControlFlowBranch(ExecuteContext=@n7.ExecutePin, Condition=@n5.Value.deformRig_r_wrist_up_alpha_79_#29)
   RigVMDispatchNode(BoneName=root, Name=deformRig_r_wrist_up_alpha, ExecuteContext=@n8.True, Value=@n6.OutputParam) {template="DISPATCH_RigDispatch_SetAnimAttribute(in Name,in BoneName,in Value,out Success)"}
   SequenceExecution_1_1 = RigVMAggregateNode(ExecuteContext=@SequenceExecution.B)
   &RigUnit_ParentConstraint(
    AdvancedSettings=(RotationOrderForFilter=XZY)
    Child=(Type=Bone,Name="deform_wrist_dn_base_l")
    Parents=((Item=(Type=Bone,Name="retarget_hand_l"),Weight=0.4),(Item=(Type=Bone,Name="retarget_lowerarm_twist_01_l"),Weight=0.6))
    Weight=1
    ExecutePin=@SequenceExecution_1_1.A
   )
   > n9 = &RigUnit_SphericalPoseReader(
    ActiveRegionScaleFactors=(PositiveWidth=1,NegativeWidth=4,PositiveHeight=3,NegativeHeight=3)
    ActiveRegionSize=0.1
    Debug=(bDrawDebug=false,DebugScale=25,DebugSegments=20,DebugThickness=0.25)
    DriverAxis=(X=1,Y=0,Z=0)
    DriverItem=(Type=Bone,Name="retarget_hand_l")
    FalloffRegionScaleFactors=(PositiveWidth=1,NegativeWidth=1,PositiveHeight=1,NegativeHeight=1)
    FalloffSize=0.4
    OptionalParentItem=(Type=Bone,Name="retarget_lowerarm_twist_01_l")
    RotationOffset=(X=-90,Y=0,Z=0)
    Debug.bDrawDebug=@n1.debug
   )
   n10 = RigVMVariableNode() {variable=value_struct_var}
   > n11 = &RigUnit_ModifyTransforms(
    ItemToModify=((Item=(Type=Bone,Name="deform_wrist_dn_l"),Transform=(Rotation=(X=0,Y=0,Z=0,W=1),Translation=(X=0,Y=0,Z=0),Scale3D=(X=1,Y=1,Z=1))))
    ItemToModify.0.Transform=@n10.Value.wrist_dn_32_#30
    Weight=@n9.OutputParam
   )
   n12 = RigVMVariableNode() {variable=curve_struct_var}
   n13 = &RigVMFunction_ControlFlowBranch(ExecuteContext=@n11.ExecutePin, Condition=@n12.Value.deformRig_l_wrist_dn_alpha_80_#31)
   RigVMDispatchNode(BoneName=root, Name=deformRig_l_wrist_dn_alpha, ExecuteContext=@n13.True, Value=@n9.OutputParam) {template="DISPATCH_RigDispatch_SetAnimAttribute(in Name,in BoneName,in Value,out Success)"}
   &RigUnit_ParentConstraint(
    AdvancedSettings=(RotationOrderForFilter=XZY)
    Child=(Type=Bone,Name="deform_wrist_dn_base_r")
    Parents=((Item=(Type=Bone,Name="retarget_hand_r"),Weight=0.4),(Item=(Type=Bone,Name="retarget_lowerarm_twist_01_r"),Weight=0.6))
    Weight=1
    ExecutePin=@SequenceExecution_1_1.B
   )
   > n14 = &RigUnit_SphericalPoseReader(
    ActiveRegionScaleFactors=(PositiveWidth=1,NegativeWidth=4,PositiveHeight=3,NegativeHeight=3)
    ActiveRegionSize=0.1
    Debug=(bDrawDebug=false,DebugScale=25,DebugSegments=20,DebugThickness=0.25)
    DriverAxis=(X=-1,Y=0,Z=0)
    DriverItem=(Type=Bone,Name="retarget_hand_r")
    FalloffRegionScaleFactors=(PositiveWidth=1,NegativeWidth=1,PositiveHeight=1,NegativeHeight=1)
    FalloffSize=0.4
    FlipHeightScaling=true
    OptionalParentItem=(Type=Bone,Name="retarget_lowerarm_twist_01_r")
    RotationOffset=(X=-90,Y=0,Z=190)
    Debug.bDrawDebug=@n1.debug
   )
   > DeformRig_GetMirrorTransform_FNC_1 = RigVMFunctionReferenceNode(transform=@n10.Value.wrist_dn_32_#30)
   > n15 = &RigUnit_ModifyTransforms(
    ItemToModify=((Item=(Type=Bone,Name="deform_wrist_dn_r"),Transform=(Rotation=(X=0,Y=0,Z=0,W=1),Translation=(X=0,Y=0,Z=0),Scale3D=(X=1,Y=1,Z=1))))
    ItemToModify.0.Transform=@DeformRig_GetMirrorTransform_FNC_1.mirrorTransform
    Weight=@n14.OutputParam
   )
   n16 = &RigVMFunction_ControlFlowBranch(ExecuteContext=@n15.ExecutePin, Condition=@n12.Value.deformRig_r_wrist_dn_alpha_81_#32)
   RigVMDispatchNode(BoneName=root, Name=deformRig_r_wrist_dn_alpha, ExecuteContext=@n16.True, Value=@n14.OutputParam) {template="DISPATCH_RigDispatch_SetAnimAttribute(in Name,in BoneName,in Value,out Success)"}
  ]}
 ContainedGraph {nodes [
   n1 = RigVMFunctionEntryNode()
   SequenceExecution = &RigVMFunction_Sequence(ExecuteContext=@n1.ExecuteContext)
   SequenceExecution_1 = &RigVMFunction_Sequence(ExecuteContext=@SequenceExecution.B)
   SequenceExecution_2 = &RigVMFunction_Sequence(ExecuteContext=@SequenceExecution_1.B)
   RigVMFunctionReturnNode(A=@SequenceExecution.A, B=@SequenceExecution_1.A, C=@SequenceExecution_2.A, D=@SequenceExecution_2.B)
  ]}
 ContainedGraph {nodes [
   n1 = RigVMFunctionEntryNode()
   SequenceExecution_1 = &RigVMFunction_Sequence(ExecuteContext=@n1.ExecuteContext)
   SequenceExecution_1_1 = &RigVMFunction_Sequence(ExecuteContext=@SequenceExecution_1.B)
   SequenceExecution_1_2 = &RigVMFunction_Sequence(ExecuteContext=@SequenceExecution_1_1.B)
   RigVMFunctionReturnNode(A=@SequenceExecution_1.A, B=@SequenceExecution_1_1.A, C=@SequenceExecution_1_2.A, D=@SequenceExecution_1_2.B)
  ]}
 ContainedGraph {nodes [
   n1 = RigVMFunctionEntryNode()
   SequenceExecution_1_1 = &RigVMFunction_Sequence(ExecuteContext=@n1.ExecuteContext)
   SequenceExecution_1_1_1 = &RigVMFunction_Sequence(ExecuteContext=@SequenceExecution_1_1.B)
   SequenceExecution_1_1_2 = &RigVMFunction_Sequence(ExecuteContext=@SequenceExecution_1_1_1.B)
   RigVMFunctionReturnNode(A=@SequenceExecution_1_1.A, B=@SequenceExecution_1_1_1.A, C=@SequenceExecution_1_1_2.A, D=@SequenceExecution_1_1_2.B)
  ]}
 ContainedGraph {nodes [
   n1 = RigVMFunctionEntryNode()
   > RigVMVariableNode(Value=@n1.value_struct) {variable=value_struct_var}
   > RigVMVariableNode(Value=@n1.curve_struct) {variable=curve_struct_var}
   > SequenceExecution = RigVMAggregateNode()
   n2 = &RigUnit_SphericalPoseReader(
    ActiveRegionScaleFactors=(PositiveWidth=4,NegativeWidth=0.5,PositiveHeight=1.5,NegativeHeight=4)
    ActiveRegionSize=0.05
    Debug=(bDrawDebug=false,DebugScale=25,DebugSegments=20,DebugThickness=0.25)
    DriverAxis=(X=1,Y=0,Z=0)
    DriverItem=(Type=Bone,Name="retarget_upperarm_l")
    FalloffRegionScaleFactors=(PositiveWidth=1,NegativeWidth=1,PositiveHeight=1,NegativeHeight=1)
    FalloffSize=0.2
    OptionalParentItem=(Type=Bone,Name="spine_03")
    RotationOffset=(X=0,Y=-150,Z=0)
    Debug.bDrawDebug=@n1.debug
    ExecutePin=@SequenceExecution.A
   )
   n3 = RigVMVariableNode() {variable=value_struct_var}
   > n4 = &RigUnit_ModifyTransforms(
    ItemToModify=((Item=(Type=Bone,Name="deform_lat_l"),Transform=(Rotation=(X=0,Y=0,Z=0,W=1),Translation=(X=0,Y=0,Z=0),Scale3D=(X=1,Y=1,Z=1))))
    Weight=@n2.OutputParam
    ItemToModify.0.Transform=@n3.Value.lat_45_#33
   )
   n5 = RigVMVariableNode() {variable=curve_struct_var}
   RigVMFunction_ControlFlowBranch = &RigVMFunction_ControlFlowBranch(ExecuteContext=@n4.ExecutePin, Condition=@n5.Value.deformRig_l_lat_alpha_91_#34)
   RigVMDispatchNode(BoneName=root, Name=deformRig_l_lat_alpha, ExecuteContext=@RigVMFunction_ControlFlowBranch.True, Value=@n2.OutputParam) {template="DISPATCH_RigDispatch_SetAnimAttribute(in Name,in BoneName,in Value,out Success)"}
   n6 = &RigUnit_SphericalPoseReader(
    ActiveRegionScaleFactors=(PositiveWidth=4,NegativeWidth=0.5,PositiveHeight=1.5,NegativeHeight=4)
    ActiveRegionSize=0.05
    Debug=(bDrawDebug=false,DebugScale=25,DebugSegments=20,DebugThickness=0.25)
    DriverAxis=(X=-1,Y=0,Z=0)
    DriverItem=(Type=Bone,Name="retarget_upperarm_r")
    FalloffRegionScaleFactors=(PositiveWidth=1,NegativeWidth=1,PositiveHeight=1,NegativeHeight=1)
    FalloffSize=0.2
    FlipHeightScaling=true
    OptionalParentItem=(Type=Bone,Name="spine_03")
    RotationOffset=(X=0,Y=30,Z=0)
    Debug.bDrawDebug=@n1.debug
    ExecutePin=@SequenceExecution.B
   )
   > DeformRig_GetMirrorTransform_FNC = RigVMFunctionReferenceNode(transform=@n3.Value.lat_45_#33)
   > n7 = &RigUnit_ModifyTransforms(
    ItemToModify=((Item=(Type=Bone,Name="deform_lat_r"),Transform=(Rotation=(X=0,Y=0,Z=0,W=1),Translation=(X=0,Y=0,Z=0),Scale3D=(X=1,Y=1,Z=1))))
    ItemToModify.0.Transform=@DeformRig_GetMirrorTransform_FNC.mirrorTransform
    Weight=@n6.OutputParam
   )
   n8 = &RigVMFunction_ControlFlowBranch(ExecuteContext=@n7.ExecutePin, Condition=@n5.Value.deformRig_r_lat_alpha_93_#35)
   RigVMDispatchNode(BoneName=root, Name=deformRig_r_lat_alpha, ExecuteContext=@n8.True, Value=@n6.OutputParam) {template="DISPATCH_RigDispatch_SetAnimAttribute(in Name,in BoneName,in Value,out Success)"}
   RigVMFunctionReturnNode(ExecuteContext=@SequenceExecution.C)
  ]}
 ContainedGraph {nodes [
   n1 = RigVMFunctionEntryNode()
   SequenceExecution = &RigVMFunction_Sequence(ExecuteContext=@n1.ExecuteContext)
   SequenceExecution_1 = &RigVMFunction_Sequence(ExecuteContext=@SequenceExecution.B)
   SequenceExecution_2 = &RigVMFunction_Sequence(ExecuteContext=@SequenceExecution_1.B)
   RigVMFunctionReturnNode(A=@SequenceExecution.A, B=@SequenceExecution_1.A, C=@SequenceExecution_2.A, D=@SequenceExecution_2.B)
  ]}
 ContainedGraph {nodes [
   n1 = RigVMFunctionEntryNode()
   SequenceAggregate_1 = &RigVMFunction_Sequence(ExecuteContext=@n1.ExecuteContext)
   SequenceAggregate_1_1 = &RigVMFunction_Sequence(ExecuteContext=@SequenceAggregate_1.B)
   SequenceAggregate_1_2 = &RigVMFunction_Sequence(ExecuteContext=@SequenceAggregate_1_1.B)
   RigVMFunctionReturnNode(A=@SequenceAggregate_1.A, B=@SequenceAggregate_1_1.A, C=@SequenceAggregate_1_2.A, D=@SequenceAggregate_1_2.B)
  ]}
 ContainedGraph {nodes [
   n1 = RigVMFunctionEntryNode()
   SequenceAggregate_2 = &RigVMFunction_Sequence(ExecuteContext=@n1.ExecuteContext)
   SequenceAggregate_2_1 = &RigVMFunction_Sequence(ExecuteContext=@SequenceAggregate_2.B)
   SequenceAggregate_2_2 = &RigVMFunction_Sequence(ExecuteContext=@SequenceAggregate_2_1.B)
   RigVMFunctionReturnNode(A=@SequenceAggregate_2.A, B=@SequenceAggregate_2_1.A, C=@SequenceAggregate_2_2.A, D=@SequenceAggregate_2_2.B)
  ]}
 New Function_ContainedGraph {nodes [
   RigVMFunctionEntryNode()
   DISPATCH_RigDispatch_GetAnimAttribute = RigVMDispatchNode(BoneName=None, Default=0, Name=deformRig_l_elbow_alpha) {template="DISPATCH_RigDispatch_GetAnimAttribute(in Name,in BoneName,in Default,out Value,out Found)"}
   > &RigUnit_ModifyTransforms(
    ItemToModify=((Item=(Type=Bone,Name="deform_tricep_l"),Transform=(Rotation=(X=0,Y=0,Z=-0.0677,W=0.9977),Translation=(X=9.3369,Y=2.7675,Z=0),Scale3D=(X=1,Y=0.3977,Z=1))))
    WeightMinimum=0
    Weight=@DISPATCH_RigDispatch_GetAnimAttribute.Value
   )
   > &RigUnit_ModifyTransforms(
    ItemToModify=((Item=(Type=Bone,Name="deform_tricep_r"),Transform=(Rotation=(X=0,Y=0,Z=-0.0677,W=0.9977),Translation=(X=-9.3369,Y=-2.7675,Z=0),Scale3D=(X=1,Y=0.3977,Z=1))))
    WeightMinimum=0
    Weight=RigVMDispatchNode(BoneName=None, Default=0, Name=deformRig_r_elbow_alpha) {template="DISPATCH_RigDispatch_GetAnimAttribute(in Name,in BoneName,in Default,out Value,out Found)"}.Value
   )
   > RigVMFunctionReturnNode()
  ]}
 New Function_ContainedGraph {nodes [
   RigVMFunctionEntryNode()
   > &RigUnit_ParentConstraint(
    AdvancedSettings=(InterpolationType=Shortest,RotationOrderForFilter=XZY)
    Child=(Type=Bone,Name="deform_scap_base_l")
    Parents=((Item=(Type=Bone,Name="spine_04"),Weight=1),(Item=(Type=Bone,Name="spine_05"),Weight=1),(Item=(Type=Bone,Name="clavicle_l"),Weight=1))
    Weight=1
   )
   > &RigUnit_ParentConstraint(
    AdvancedSettings=(InterpolationType=Shortest,RotationOrderForFilter=XZY)
    Child=(Type=Bone,Name="deform_scap_base_r")
    Parents=((Item=(Type=Bone,Name="spine_04"),Weight=1),(Item=(Type=Bone,Name="spine_05"),Weight=1),(Item=(Type=Bone,Name="clavicle_r"),Weight=1))
    Weight=1
   )
   > n1 = &RigUnit_SphericalPoseReader(
    ActiveRegionScaleFactors=(PositiveWidth=0,NegativeWidth=3,PositiveHeight=3,NegativeHeight=3)
    ActiveRegionSize=0.1
    Debug=(bDrawDebug=True,bDraw2D=False,bDrawLocalAxes=False,DebugScale=25,DebugSegments=20,DebugThickness=0.25)
    DriverAxis=(X=1,Y=0,Z=0)
    DriverItem=(Type=Bone,Name="clavicle_l")
    FalloffRegionScaleFactors=(PositiveWidth=1,NegativeWidth=1,PositiveHeight=1,NegativeHeight=1)
    FalloffSize=0.33
    FlipHeightScaling=False
    FlipWidthScaling=False
    OptionalParentItem=(Type=Bone,Name="spine_05")
    RotationOffset=(X=-90,Y=0,Z=20)
   )
   > &RigUnit_ModifyTransforms(
    ItemToModify=((Item=(Type=Bone,Name="deform_scap_up_l"),Transform=(Rotation=(X=0,Y=0,Z=0.4241,W=-0.9056),Translation=(X=-5.0704,Y=-16.0841,Z=0),Scale3D=(X=1,Y=1,Z=1))))
    WeightMinimum=0
    Weight=@n1.OutputParam
   )
   > n2 = &RigUnit_SphericalPoseReader(
    ActiveRegionScaleFactors=(PositiveWidth=0,NegativeWidth=3,PositiveHeight=3,NegativeHeight=3)
    ActiveRegionSize=0.1
    Debug=(bDrawDebug=True,bDraw2D=False,bDrawLocalAxes=False,DebugScale=25,DebugSegments=20,DebugThickness=0.25)
    DriverAxis=(X=-1,Y=0,Z=0)
    DriverItem=(Type=Bone,Name="clavicle_r")
    FalloffRegionScaleFactors=(PositiveWidth=1,NegativeWidth=1,PositiveHeight=1,NegativeHeight=1)
    FalloffSize=0.33
    FlipHeightScaling=False
    FlipWidthScaling=true
    OptionalParentItem=(Type=Bone,Name="spine_05")
    RotationOffset=(X=90,Y=0,Z=20)
   )
   > &RigUnit_ModifyTransforms(
    ItemToModify=((Item=(Type=Bone,Name="deform_scap_up_r"),Transform=(Rotation=(X=0,Y=0,Z=0.4241,W=-0.9056),Translation=(X=-5.0704,Y=-16.0841,Z=0),Scale3D=(X=1,Y=1,Z=1))))
    WeightMinimum=0
    Weight=@n2.OutputParam
   )
   > RigVMFunctionReturnNode()
  ]}
 New Function_ContainedGraph {nodes [
   RigVMFunctionEntryNode()
   > n1 = &RigUnit_SphericalPoseReader(
    ActiveRegionScaleFactors=(PositiveWidth=0,NegativeWidth=4,PositiveHeight=4,NegativeHeight=4)
    ActiveRegionSize=0.1
    Debug=(bDrawDebug=True,bDraw2D=False,bDrawLocalAxes=False,DebugScale=25,DebugSegments=20,DebugThickness=0.25)
    DriverAxis=(X=1,Y=0,Z=0)
    DriverItem=(Type=Bone,Name="spine_04")
    FalloffRegionScaleFactors=(PositiveWidth=1,NegativeWidth=1,PositiveHeight=1,NegativeHeight=1)
    FalloffSize=0.4
    FlipHeightScaling=False
    FlipWidthScaling=False
    OptionalParentItem=(Type=Bone,Name="pelvis")
    RotationOffset=(X=90,Y=0,Z=0)
   )
   > &RigUnit_ModifyTransforms(
    ItemToModify=((Item=(Type=Bone,Name="deform_stomach"),Transform=(Rotation=(X=0,Y=0,Z=0,W=-1),Translation=(X=0,Y=13.5123,Z=0),Scale3D=(X=1.5625,Y=1,Z=2.3125))))
    WeightMinimum=0
    Weight=@n1.OutputParam
   )
   > RigVMFunctionReturnNode()
  ]}
 New Function_ContainedGraph {nodes [
   RigVMFunctionEntryNode()
   > Get Normalized Distance Between = RigVMFunctionReferenceNode(Min Distance=0, Source=(Type=None,Name="None"), Target=(Type=None,Name="None"))
   > RigVMVariableNode(Value=@"Get Normalized Distance Between".Result) {variable=LeftArmCloseToSternum}
   > Get Normalized Distance Between_1 = RigVMFunctionReferenceNode(Min Distance=0, Source=(Type=None,Name="None"), Target=(Type=None,Name="None"))
   > RigVMVariableNode(Value=@"Get Normalized Distance Between_1".Result) {variable=RightArmCloseToSternum}
   > RigVMFunctionReturnNode()
  ]}
 New Function_ContainedGraph {nodes [
   n1 = RigVMFunctionEntryNode()
   > RigVMVariableNode(Value=@n1.DeformRig_Bool) {variable=deform_bool_struct}
   > RigVMVariableNode(Value=@n1.DeformRig_Value) {variable=deform_value_struct}
   > RigVMVariableNode(Value=@n1.DeformRig_ Curve) {variable=deform_curve_struct}
   > RigVMVariableNode(Value=@n1.DeformRig_Debug) {variable=deform_debug_struct}
   > SequenceAggregate = RigVMAggregateNode()
   SequenceAggregate_1 = RigVMAggregateNode(ExecuteContext=@SequenceAggregate.A)
   n2 = &RigVMFunction_ControlFlowBranch(
    Condition=RigVMVariableNode() {variable=deform_bool_struct}.Value.upperarm_1_#36
    ExecuteContext=@SequenceAggregate_1.A
   )
   DeformRig_UpperArm_FNC = RigVMFunctionReferenceNode(
    ExecuteContext=@n2.True
    value_struct=RigVMVariableNode() {variable=deform_value_struct}.Value
    debug=RigVMVariableNode() {variable=deform_debug_struct}.Value.upperarm_debug_18_#36
    curve_struct=RigVMVariableNode() {variable=deform_curve_struct}.Value
   )
   n3 = &RigVMFunction_ControlFlowBranch(
    Condition=RigVMVariableNode() {variable=deform_bool_struct}.Value.wrist_7_#37
    ExecuteContext=@SequenceAggregate_1.B
   )
   DeformRig_Wrist_FNC = RigVMFunctionReferenceNode(
    ExecuteContext=@n3.True
    value_struct=RigVMVariableNode() {variable=deform_value_struct}.Value
    debug=RigVMVariableNode() {variable=deform_debug_struct}.Value.wrist_debug_20_#37
    curve_struct=RigVMVariableNode() {variable=deform_curve_struct}.Value
   )
   n4 = &RigVMFunction_ControlFlowBranch(
    Condition=RigVMVariableNode() {variable=deform_bool_struct}.Value.knee_9_#38
    ExecuteContext=@SequenceAggregate_1.C
   )
   DeformRig_Knee_FNC = RigVMFunctionReferenceNode(
    ExecuteContext=@n4.True
    value_struct=RigVMVariableNode() {variable=deform_value_struct}.Value
    curve_struct=RigVMVariableNode() {variable=deform_curve_struct}.Value
   )
   n5 = &RigVMFunction_ControlFlowBranch(
    Condition=RigVMVariableNode() {variable=deform_bool_struct}.Value.pec_12_#39
    ExecuteContext=@SequenceAggregate_1.D
   )
   DeformRig_Pec_FNC = RigVMFunctionReferenceNode(
    ExecuteContext=@n5.True
    value_struct=RigVMVariableNode() {variable=deform_value_struct}.Value
    curve_struct=RigVMVariableNode() {variable=deform_curve_struct}.Value
   )
   SequenceAggregate_2 = RigVMAggregateNode(ExecuteContext=@SequenceAggregate.B)
   n6 = &RigVMFunction_ControlFlowBranch(
    Condition=RigVMVariableNode() {variable=deform_bool_struct}.Value.lat_15_#40
    ExecuteContext=@SequenceAggregate_2.A
   )
   DeformRig_Lat_FNC = RigVMFunctionReferenceNode(
    ExecuteContext=@n6.True
    value_struct=RigVMVariableNode() {variable=deform_value_struct}.Value
    debug=RigVMVariableNode() {variable=deform_debug_struct}.Value.lat_debug_26_#40
    curve_struct=RigVMVariableNode() {variable=deform_curve_struct}.Value
   )
   n7 = RigVMVariableNode() {variable=deform_bool_struct}
   n8 = &RigVMFunction_ControlFlowBranch(
    Condition=&RigVMFunction_MathBoolOr(B=@n7.Value.bicep_13_#41, A=@n7.Value.elbow_3_#42).Result
    ExecuteContext=@SequenceAggregate_2.B
   )
   DeformRig_Elbow_FNC = RigVMFunctionReferenceNode(
    ExecuteContext=@n8.True
    value_struct=RigVMVariableNode() {variable=deform_value_struct}.Value
    curve_struct=RigVMVariableNode() {variable=deform_curve_struct}.Value
    bicep=@n7.Value.bicep_13_#41
    elbow=@n7.Value.elbow_3_#42
   )
   > &RigUnit_ParentConstraint(
    AdvancedSettings=(InterpolationType=Shortest,RotationOrderForFilter=XZY)
    Child=(Type=Bone,Name="deform_elbow_in_base_l")
    Parents=((Item=(Type=Bone,Name="retarget_lowerarm_l"),Weight=0.4),(Item=(Type=Bone,Name="retarget_upperarm_l"),Weight=0.6))
    Weight=1
   )
   > &RigUnit_ParentConstraint(
    AdvancedSettings=(RotationOrderForFilter=XZY)
    Child=(Type=Bone,Name="deform_elbow_in_base_r")
    Parents=((Item=(Type=Bone,Name="retarget_lowerarm_r"),Weight=0.4),(Item=(Type=Bone,Name="retarget_upperarm_r"),Weight=0.6))
    Weight=1
   )
   > &RigUnit_ParentConstraint(
    AdvancedSettings=(RotationOrderForFilter=XZY)
    Child=(Type=Bone,Name="deform_elbow_out_base_l")
    Parents=((Item=(Type=Bone,Name="retarget_lowerarm_l"),Weight=0.3),(Item=(Type=Bone,Name="retarget_upperarm_l"),Weight=0.7))
    Weight=1
   )
   > &RigUnit_ParentConstraint(
    AdvancedSettings=(RotationOrderForFilter=XZY)
    Child=(Type=Bone,Name="deform_elbow_out_base_r")
    Parents=((Item=(Type=Bone,Name="retarget_lowerarm_r"),Weight=0.3),(Item=(Type=Bone,Name="retarget_upperarm_r"),Weight=0.7))
    Weight=1
    bMaintainOffset=true
   )
   n9 = RigVMVariableNode() {variable=deform_bool_struct}
   n10 = &RigVMFunction_ControlFlowBranch(
    Condition=&RigVMFunction_MathBoolOr(B=@n9.Value.glute_17_#43, A=@n9.Value.groin_8_#44).Result
    ExecuteContext=@SequenceAggregate_2.C
   )
   DeformRig_GroinGlue_FNC = RigVMFunctionReferenceNode(
    ExecuteContext=@n10.True
    value_struct=RigVMVariableNode() {variable=deform_value_struct}.Value
    debug=RigVMVariableNode() {variable=deform_debug_struct}.Value.groinGlute_debug_27_#44
    curve_struct=RigVMVariableNode() {variable=deform_curve_struct}.Value
    glute=@n9.Value.glute_17_#43
    groin=@n9.Value.groin_8_#44
   )
   RigVMFunctionReturnNode(ExecuteContext=@SequenceAggregate.C)
  ]}
 New Function_ContainedGraph {nodes [
   n1 = RigVMFunctionEntryNode()
   RigVMFunction_Sequence = &RigVMFunction_Sequence(ExecuteContext=@n1.ExecuteContext)
   RigVMFunction_Sequence_1 = RigVMAggregateNode(ExecuteContext=@RigVMFunction_Sequence.A)
   RigVMFunction_ControlFlowBranch = &RigVMFunction_ControlFlowBranch(ExecuteContext=@RigVMFunction_Sequence_1.A, Condition=RigVMVariableNode() {variable=DisableRuntimeArmTwists}.Value)
   RigVMFunction_Sequence_2 = RigVMAggregateNode(ExecuteContext=@RigVMFunction_ControlFlowBranch.False)
   RigVMFunction_MathQuaternionSwingTwist = &RigVMFunction_MathQuaternionSwingTwist(
    TwistAxis=(X=1,Y=0,Z=0)
    Input=&RigVMFunction_MathQuaternionMakeRelative(
     Global=&RigUnit_GetTransform(Item=(Type=Bone,Name="retarget_upperarm_l"), Space=LocalSpace, bInitial=False).Transform.Rotation
     Parent=&RigUnit_GetTransform(Item=(Type=Bone,Name="retarget_upperarm_l"), Space=LocalSpace, bInitial=true).Transform.Rotation
    ) {template="Make Relative::Execute(in Global,in Parent,out Local)"}.Local
   )
   Inverse = /Script/RigVM.RigVMFunction_MathQuaternionInverse(Value=@RigVMFunction_MathQuaternionSwingTwist.Twist) {is=struct RigVMUnitNode, template="Inverse::Execute(in Value,out Result)"}
   Set Transform = &RigUnit_SetRotation(
    Item=(Type=Bone,Name="retarget_upperarm_twist_01_l")
    Weight=0.7
    bInitial=false
    Value=@Inverse.Result
    ExecutePin=@RigVMFunction_Sequence_2.A
   )
   > Set Transform_1 = &RigUnit_SetRotation(Item=(Type=Bone,Name="retarget_upperarm_twist_02_l"), Weight=0.3, bInitial=false, Value=@Inverse.Result)
   n2 = &RigVMFunction_MathQuaternionMakeRelative(
    Global=&RigUnit_GetTransform(Item=(Type=Bone,Name="retarget_upperarm_r"), Space=LocalSpace, bInitial=False).Transform.Rotation
    Parent=&RigUnit_GetTransform(Item=(Type=Bone,Name="retarget_upperarm_r"), Space=LocalSpace, bInitial=true).Transform.Rotation
   ) {template="Make Relative::Execute(in Global,in Parent,out Local)"}
   Set Transform_2 = &RigUnit_SetRotation(
    Item=(Type=Bone,Name="retarget_upperarm_twist_01_r")
    Weight=0.7
    bInitial=false
    Value=&RigVMFunction_MathQuaternionInverse(Value=&RigVMFunction_MathQuaternionSwingTwist(TwistAxis=(X=-1,Y=0,Z=0), Input=@n2.Local).Twist) {template="Inverse::Execute(in Value,out Result)"}.Result
    ExecutePin=@RigVMFunction_Sequence_2.B
   )
   > Set Transform_1_2 = &RigUnit_SetRotation(Item=(Type=Bone,Name="retarget_upperarm_twist_02_r"), Value=(X=0,Y=0,Z=0,W=1), Weight=0.3, bInitial=false)
   n3 = &RigUnit_GetTransform(Item=(Type=Bone,Name="retarget_hand_l"), Space=LocalSpace, bInitial=False)
   n4 = &RigUnit_GetTransform(Item=(Type=Bone,Name="retarget_hand_l"), Space=LocalSpace, bInitial=true)
   Set Transform_1_1_3 = &RigUnit_SetRotation(
    Item=(Type=Bone,Name="retarget_lowerarm_twist_01_l")
    Weight=0.5
    bInitial=false
    Value=&RigVMFunction_MathQuaternionSwingTwist(
     TwistAxis=(X=1,Y=0,Z=0)
     Input=&RigVMFunction_MathQuaternionMakeRelative(Global=@n3.Transform.Rotation, Parent=@n4.Transform.Rotation) {template="Make Relative::Execute(in Global,in Parent,out Local)"}.Local
    ).Twist
    ExecutePin=@RigVMFunction_Sequence_2.C
   )
   RigUnit_GetTransform = &RigUnit_GetTransform(Item=(Type=Bone,Name="retarget_lowerarm_twist_01_l"), Space=LocalSpace, bInitial=False)
   > Set Transform_1_1_3_1 = &RigUnit_SetRotation(Item=(Type=Bone,Name="retarget_lowerarm_twist_02_l"), Weight=0.3, bInitial=false, Value=@RigUnit_GetTransform.Transform.Rotation)
   n5 = &RigUnit_GetTransform(Item=(Type=Bone,Name="retarget_hand_r"), Space=LocalSpace, bInitial=False)
   n6 = &RigUnit_GetTransform(Item=(Type=Bone,Name="retarget_hand_r"), Space=LocalSpace, bInitial=true)
   Set Transform_1_1_3_2 = &RigUnit_SetRotation(
    Item=(Type=Bone,Name="retarget_lowerarm_twist_01_r")
    Weight=0.5
    bInitial=false
    Value=&RigVMFunction_MathQuaternionSwingTwist(
     TwistAxis=(X=1,Y=0,Z=0)
     Input=&RigVMFunction_MathQuaternionMakeRelative(Global=@n5.Transform.Rotation, Parent=@n6.Transform.Rotation) {template="Make Relative::Execute(in Global,in Parent,out Local)"}.Local
    ).Twist
    ExecutePin=@RigVMFunction_Sequence_2.D
   )
   > Set Transform_1_1_3_1_1 = &RigUnit_SetRotation(
    Item=(Type=Bone,Name="retarget_lowerarm_twist_02_r")
    Weight=0.3
    bInitial=false
    Value=&RigUnit_GetTransform(Item=(Type=Bone,Name="retarget_lowerarm_twist_01_r"), Space=LocalSpace, bInitial=False).Transform.Rotation
   )
   &RigUnit_ParentConstraint(
    AdvancedSettings=(InterpolationType=Average,RotationOrderForFilter=XZY)
    Child=(Type=Bone,Name="retarget_upperarm_twist_01_l")
    Parents=((Item=(Type=Bone,Name="upperarm_twist_01_l"),Weight=1))
    Weight=1
    bMaintainOffset=true
    ExecutePin=@RigVMFunction_ControlFlowBranch.True
   )
   > &RigUnit_ParentConstraint(
    AdvancedSettings=(InterpolationType=Average,RotationOrderForFilter=XZY)
    Child=(Type=Bone,Name="retarget_upperarm_twist_02_l")
    Parents=((Item=(Type=Bone,Name="upperarm_twist_02_l"),Weight=1))
    Weight=1
    bMaintainOffset=true
   )
   > &RigUnit_ParentConstraint(
    AdvancedSettings=(InterpolationType=Average,RotationOrderForFilter=XZY)
    Child=(Type=Bone,Name="retarget_lowerarm_twist_01_l")
    Parents=((Item=(Type=Bone,Name="lowerarm_twist_01_l"),Weight=1))
    Weight=1
    bMaintainOffset=true
   )
   > &RigUnit_ParentConstraint(
    AdvancedSettings=(InterpolationType=Average,RotationOrderForFilter=XZY)
    Child=(Type=Bone,Name="retarget_lowerarm_twist_02_l")
    Parents=((Item=(Type=Bone,Name="lowerarm_twist_02_l"),Weight=1))
    Weight=1
   )
   > &RigUnit_ParentConstraint(
    AdvancedSettings=(InterpolationType=Average,RotationOrderForFilter=XZY)
    Child=(Type=Bone,Name="retarget_upperarm_twist_01_r")
    Parents=((Item=(Type=Bone,Name="upperarm_twist_01_r"),Weight=1))
    Weight=1
   )
   > &RigUnit_ParentConstraint(
    AdvancedSettings=(InterpolationType=Average,RotationOrderForFilter=XZY)
    Child=(Type=Bone,Name="retarget_lowerarm_twist_02_r")
    Parents=((Item=(Type=Bone,Name="upperarm_twist_02_r"),Weight=1))
    Weight=1
   )
   > &RigUnit_ParentConstraint(
    AdvancedSettings=(InterpolationType=Average,RotationOrderForFilter=XZY)
    Child=(Type=Bone,Name="retarget_lowerarm_twist_01_r")
    Parents=((Item=(Type=Bone,Name="lowerarm_twist_01_r"),Weight=1))
    Weight=1
   )
   > &RigUnit_ParentConstraint(
    AdvancedSettings=(InterpolationType=Average,RotationOrderForFilter=XZY)
    Child=(Type=Bone,Name="retarget_lowerarm_twist_02_r")
    Parents=((Item=(Type=Bone,Name="lowerarm_twist_02_r"),Weight=1))
    Weight=1
   )
   n7 = &RigUnit_GetTransform(Item=(Type=Bone,Name="retarget_thigh_l"), Space=LocalSpace, bInitial=true)
   n8 = &RigUnit_GetTransform(Item=(Type=Bone,Name="retarget_thigh_l"), Space=LocalSpace, bInitial=False)
   n9 = &RigVMFunction_MathQuaternionInverse(Value=&RigVMFunction_MathQuaternionSwingTwist(
     TwistAxis=(X=1,Y=0,Z=0)
     Input=&RigVMFunction_MathQuaternionMakeRelative(Parent=@n7.Transform.Rotation, Global=@n8.Transform.Rotation) {template="Make Relative::Execute(in Global,in Parent,out Local)"}.Local
    ).Twist) {template="Inverse::Execute(in Value,out Result)"}
   Set Transform_3 = &RigUnit_SetRotation(Item=(Type=Bone,Name="retarget_thigh_twist_01_l"), Weight=1, bInitial=false, Value=@n9.Result, ExecutePin=@RigVMFunction_Sequence_1.B)
   n10 = &RigVMFunction_MathQuaternionMakeRelative(
    Parent=&RigUnit_GetTransform(Item=(Type=Bone,Name="retarget_thigh_r"), Space=LocalSpace, bInitial=true).Transform.Rotation
    Global=&RigUnit_GetTransform(Item=(Type=Bone,Name="retarget_thigh_r"), Space=LocalSpace, bInitial=False).Transform.Rotation
   ) {template="Make Relative::Execute(in Global,in Parent,out Local)"}
   Set Transform_2_1 = &RigUnit_SetRotation(
    Item=(Type=Bone,Name="retarget_thigh_twist_01_r")
    Weight=1
    bInitial=false
    Value=&RigVMFunction_MathQuaternionInverse(Value=&RigVMFunction_MathQuaternionSwingTwist(TwistAxis=(X=-1,Y=0,Z=0), Input=@n10.Local).Twist) {template="Inverse::Execute(in Value,out Result)"}.Result
    ExecutePin=@RigVMFunction_Sequence_1.C
   )
   > Set Transform_1_2_1 = &RigUnit_SetRotation(Item=(Type=Bone,Name="retarget_upperarm_twist_02_r"), Value=(X=0,Y=0,Z=0,W=1), Weight=0.5, bInitial=false)
   RigVMFunctionReturnNode(ExecuteContext=@RigVMFunction_Sequence.B)
   &RigUnit_RotationConstraintLocalSpaceOffset(
    AdvancedSettings=(InterpolationType=Average,RotationOrderForFilter=XZY)
    Child=(Type=Bone,Name="retarget_hand_split_l")
    ExecutePin=()
    Filter=(bX=True,bY=True,bZ=True)
    Parents=((Item=(Type=Bone,Name="retarget_hand_l"),Weight=1))
    Weight=0.5
   )
   > &RigUnit_RotationConstraintLocalSpaceOffset(
    AdvancedSettings=(InterpolationType=Average,RotationOrderForFilter=XZY)
    Child=(Type=Bone,Name="retarget_hand_split_r")
    Filter=(bX=True,bY=True,bZ=True)
    Parents=((Item=(Type=Bone,Name="retarget_hand_r"),Weight=1))
    Weight=0.5
   )
   &RigUnit_ModifyTransforms(
    ExecutePin=()
    ItemToModify=((Item=(Type=Bone,Name="retarget_lowerarm_twist_01_l"),Transform=(Rotation=(X=-1,Y=0,Z=0,W=0),Translation=(X=0,Y=0,Z=0),Scale3D=(X=1,Y=1,Z=1))))
    Weight=0
    WeightMinimum=0
   )
   &RigUnit_ModifyTransforms(
    ExecutePin=()
    ItemToModify=((Item=(Type=Bone,Name="retarget_lowerarm_twist_01_r"),Transform=(Rotation=(X=-1,Y=0,Z=0,W=0),Translation=(X=0,Y=0,Z=0),Scale3D=(X=1,Y=1,Z=1))))
    Weight=0
    WeightMinimum=0
   )
   Set Transform_1_1 = &RigUnit_SetRotation(ExecutePin=(), Item=(Type=Bone,Name="retarget_upperarm_twist_02_l"), Weight=0.5, bInitial=false, Value=@n9.Result)
  ]}
 New Function_ContainedGraph {nodes [
   RigVMFunctionEntryNode()
   > n1 = &RigUnit_SphericalPoseReader(
    ActiveRegionScaleFactors=(PositiveWidth=1,NegativeWidth=4,PositiveHeight=3,NegativeHeight=1.5)
    ActiveRegionSize=0.1
    Debug=(bDrawDebug=false,bDraw2D=False,bDrawLocalAxes=False,DebugScale=25,DebugSegments=20,DebugThickness=0.25)
    DriverAxis=(X=1,Y=0,Z=0)
    DriverItem=(Type=Bone,Name="retarget_upperarm_l")
    FalloffRegionScaleFactors=(PositiveWidth=1,NegativeWidth=1,PositiveHeight=1,NegativeHeight=1)
    FalloffSize=0.35
    FlipHeightScaling=False
    FlipWidthScaling=False
    OptionalParentItem=(Type=Bone,Name="spine_05")
    RotationOffset=(X=110,Y=0,Z=45)
   )
   > &RigUnit_ModifyTransforms(
    ItemToModify="((Item=(Type=Bone,Name=\"deform_lat_l\"),Transform=(Rotation=(X=0.243,Y=0.1157,Z=0.0292,W=0.9627),Translation=(X=0.7867,Y=13.1417,Z=-9.4179),Scale3D=(X=1,Y=1,Z=1))),(Item=(Type=Bone,Name=\"deform_scap_up_l\"),Transform=(Rotation=(X=0,Y=0,Z=0,W=-1),Translation= ...(truncated)"
    WeightMinimum=0
    Weight=@n1.OutputParam
   )
   > n2 = &RigUnit_SphericalPoseReader(
    ActiveRegionScaleFactors=(PositiveWidth=1,NegativeWidth=4,PositiveHeight=3,NegativeHeight=1.5)
    ActiveRegionSize=0.1
    Debug=(bDrawDebug=true,bDraw2D=False,bDrawLocalAxes=False,DebugScale=25,DebugSegments=20,DebugThickness=0.25)
    DriverAxis=(X=-1,Y=0,Z=0)
    DriverItem=(Type=Bone,Name="retarget_upperarm_r")
    FalloffRegionScaleFactors=(PositiveWidth=1,NegativeWidth=1,PositiveHeight=1,NegativeHeight=1)
    FalloffSize=0.4
    FlipHeightScaling=true
    FlipWidthScaling=true
    OptionalParentItem=(Type=Bone,Name="spine_05")
    RotationOffset=(X=-90,Y=45,Z=60)
   )
   > &RigUnit_ModifyTransforms(
    ItemToModify="((Item=(Type=Bone,Name=\"deform_lat_r\"),Transform=(Rotation=(X=0.243,Y=0.1157,Z=0.0292,W=0.9627),Translation=(X=-0.7867,Y=-13.1417,Z=10),Scale3D=(X=1,Y=1,Z=1))),(Item=(Type=Bone,Name=\"deform_scap_up_r\"),Transform=(Rotation=(X=0,Y=0,Z=0,W=-1),Translation=(X= ...(truncated)"
    WeightMinimum=0
    Weight=@n2.OutputParam
   )
   > RigVMFunctionReturnNode()
  ]}
 RigVMFunctionLibrary {nodes [
   Calculate Arm Distance To Sternum = RigVMCollapseNode(ExecuteContext=())
   Deform Lat Extra = RigVMCollapseNode(ExecuteContext=())
   Deform Scapula = RigVMCollapseNode(ExecuteContext=())
   Deform Stomach = RigVMCollapseNode(ExecuteContext=())
   Deform Tricep = RigVMCollapseNode(ExecuteContext=())
   Deform Twists = RigVMCollapseNode(ExecuteContext=())
   DeformRig_Elbow_FNC = RigVMCollapseNode(
    ExecuteContext=()
    bicep=true
    curve_struct=(deformRig_l_lat_alpha_91_#34=False,deformRig_r_lat_alpha_93_#35=False)
    elbow=true
    value_struct="(upperarm_up_18_#21=(Rotation=(X=0,Y=0,Z=0,W=1),Translation=(X=0,Y=0,Z=8),Scale3D=(X=0.7,Y=1,Z=1)),upperarm_fwd_fr_19_#14=(Rotation=(X=0,Y=0,Z=0,W=1),Translation=(X=0,Y=10,Z=0),Scale3D=(X=0.5,Y=1,Z= ...(truncated)"
   )
   DeformRig_GetDeltaRefTransform_FNC = RigVMCollapseNode(ExecuteContext=(), Item=(Type=Bone))
   DeformRig_GetDeltaRefTransform_FNC_1 = RigVMCollapseNode(ExecuteContext=(), Item=(Type=Bone))
   DeformRig_GetMirrorTransform_FNC = RigVMCollapseNode(ExecuteContext=(), transform=(Rotation=(X=0,Y=0,Z=0,W=1),Translation=(X=0,Y=0,Z=0),Scale3D=(X=1,Y=1,Z=1)))
   DeformRig_GetMirrorTransform_FNC_1 = RigVMCollapseNode(ExecuteContext=(), transform=(Rotation=(X=0,Y=0,Z=0,W=1),Translation=(X=0,Y=0,Z=0),Scale3D=(X=1,Y=1,Z=1)))
   DeformRig_GetMirrorTransform_FNC_2 = RigVMCollapseNode(ExecuteContext=(), transform=(Rotation=(X=0,Y=0,Z=0,W=1),Translation=(X=0,Y=0,Z=0),Scale3D=(X=1,Y=1,Z=1)))
   DeformRig_GroinGlue_FNC = RigVMCollapseNode(
    ExecuteContext=()
    curve_struct=(deformRig_l_lat_alpha_91_#34=False,deformRig_r_lat_alpha_93_#35=False)
    debug=False
    glute=true
    groin=true
    value_struct="(upperarm_up_18_#21=(Rotation=(X=0,Y=0,Z=0,W=1),Translation=(X=0,Y=0,Z=8),Scale3D=(X=0.7,Y=1,Z=1)),upperarm_fwd_fr_19_#14=(Rotation=(X=0,Y=0,Z=0,W=1),Translation=(X=0,Y=10,Z=0),Scale3D=(X=0.5,Y=1,Z= ...(truncated)"
   )
   DeformRig_GroinGlue_FNC_1 = RigVMCollapseNode(
    ExecuteContext=()
    curve_struct=(deformRig_l_lat_alpha_91_#34=False,deformRig_r_lat_alpha_93_#35=False)
    debug=False
    glute=true
    groin=true
    value_struct="(upperarm_up_18_#21=(Rotation=(X=0,Y=0,Z=0,W=1),Translation=(X=0,Y=0,Z=8),Scale3D=(X=0.7,Y=1,Z=1)),upperarm_fwd_fr_19_#14=(Rotation=(X=0,Y=0,Z=0,W=1),Translation=(X=0,Y=10,Z=0),Scale3D=(X=0.5,Y=1,Z= ...(truncated)"
   )
   DeformRig_Knee_FNC = RigVMCollapseNode(
    ExecuteContext=()
    curve_struct=(deformRig_l_lat_alpha_91_#34=False,deformRig_r_lat_alpha_93_#35=False)
    value_struct="(upperarm_up_18_#21=(Rotation=(X=0,Y=0,Z=0,W=1),Translation=(X=0,Y=0,Z=8),Scale3D=(X=0.7,Y=1,Z=1)),upperarm_fwd_fr_19_#14=(Rotation=(X=0,Y=0,Z=0,W=1),Translation=(X=0,Y=10,Z=0),Scale3D=(X=0.5,Y=1,Z= ...(truncated)"
   )
   DeformRig_Knee_FNC_1 = RigVMCollapseNode(
    ExecuteContext=()
    curve_struct=(deformRig_l_lat_alpha_91_#34=False,deformRig_r_lat_alpha_93_#35=False)
    value_struct="(upperarm_up_18_#21=(Rotation=(X=0,Y=0,Z=0,W=1),Translation=(X=0,Y=0,Z=8),Scale3D=(X=0.7,Y=1,Z=1)),upperarm_fwd_fr_19_#14=(Rotation=(X=0,Y=0,Z=0,W=1),Translation=(X=0,Y=10,Z=0),Scale3D=(X=0.5,Y=1,Z= ...(truncated)"
   )
   DeformRig_Lat_FNC = RigVMCollapseNode(
    ExecuteContext=()
    curve_struct=(deformRig_l_lat_alpha_91_#34=False,deformRig_r_lat_alpha_93_#35=False)
    debug=False
    value_struct="(upperarm_up_18_#21=(Rotation=(X=0,Y=0,Z=0,W=1),Translation=(X=0,Y=0,Z=8),Scale3D=(X=0.7,Y=1,Z=1)),upperarm_fwd_fr_19_#14=(Rotation=(X=0,Y=0,Z=0,W=1),Translation=(X=0,Y=10,Z=0),Scale3D=(X=0.5,Y=1,Z= ...(truncated)"
   )
   DeformRig_Pec_FNC = RigVMCollapseNode(
    ExecuteContext=()
    curve_struct=()
    value_struct="(upperarm_up_18_#21=(Rotation=(X=0,Y=0,Z=0,W=1),Translation=(X=0,Y=0,Z=8),Scale3D=(X=0.7,Y=1,Z=1)),upperarm_fwd_fr_19_#14=(Rotation=(X=0,Y=0,Z=0,W=1),Translation=(X=0,Y=10,Z=0),Scale3D=(X=0.5,Y=1,Z= ...(truncated)"
   )
   DeformRig_UpperArm_FNC = RigVMCollapseNode(
    ExecuteContext=()
    curve_struct=(deformRig_l_lat_alpha_91_#34=False,deformRig_r_lat_alpha_93_#35=False)
    debug=False
    value_struct="(upperarm_up_18_#21=(Rotation=(X=0,Y=0,Z=0,W=1),Translation=(X=0,Y=0,Z=8),Scale3D=(X=0.7,Y=1,Z=1)),upperarm_fwd_fr_19_#14=(Rotation=(X=0,Y=0,Z=0,W=1),Translation=(X=0,Y=10,Z=0),Scale3D=(X=0.5,Y=1,Z= ...(truncated)"
   )
   DeformRig_Wrist_FNC = RigVMCollapseNode(
    ExecuteContext=()
    curve_struct=(deformRig_l_lat_alpha_91_#34=False,deformRig_r_lat_alpha_93_#35=False)
    debug=false
    value_struct="(upperarm_up_18_#21=(Rotation=(X=0,Y=0,Z=0,W=1),Translation=(X=0,Y=0,Z=8),Scale3D=(X=0.7,Y=1,Z=1)),upperarm_fwd_fr_19_#14=(Rotation=(X=0,Y=0,Z=0,W=1),Translation=(X=0,Y=10,Z=0),Scale3D=(X=0.5,Y=1,Z= ...(truncated)"
   )
   FN_DeformRig_FNC = RigVMCollapseNode(
    DeformRig_ Curve=()
    DeformRig_Bool="(upperarm_1_#36=True,elbow_3_#42=True,wrist_7_#37=True,groin_8_#44=True,glute_17_#43=True,knee_9_40C408324CA261 ...(truncated)"
    DeformRig_Debug=(upperarm_debug_18_#36=False,wrist_debug_20_#37=False,groinGlute_debug_27_#44=False,lat_debug_26_#40=False)
    DeformRig_Value="(upperarm_up_18_#21=(Rotation=(X=0,Y=0,Z=0,W=1),Translation=(X=0,Y=0,Z=8),Scale3D=(X=0.7,Y=1,Z=1)),upperarm_fwd_fr_19_#14=(Rotation=(X=0,Y=0,Z=0,W=1),Translation=(X=0,Y=10,Z=0),Scale3D=(X=0.5,Y=1,Z= ...(truncated)"
    ExecuteContext=()
   )
  ]}
 RigVMFunction_Sequence_1_1_ContainedGraph {nodes [
   n1 = RigVMFunctionEntryNode()
   n2 = &RigVMFunction_Sequence(ExecuteContext=@n1.ExecuteContext)
   n3 = &RigVMFunction_Sequence(ExecuteContext=@n2.B)
   RigVMFunctionReturnNode(A=@n2.A, B=@n3.A, C=@n3.B)
  ]}
 RigVMFunction_Sequence_2_1_ContainedGraph {nodes [
   n1 = RigVMFunctionEntryNode()
   n2 = &RigVMFunction_Sequence(ExecuteContext=@n1.ExecuteContext)
   n3 = &RigVMFunction_Sequence(ExecuteContext=@n2.B)
   n4 = &RigVMFunction_Sequence(ExecuteContext=@n3.B)
   RigVMFunctionReturnNode(A=@n2.A, B=@n3.A, C=@n4.A, D=@n4.B)
  ]}
 RigVMModel {nodes [
   RigUnit_BeginExecution = &RigUnit_BeginExecution()
   Deform Twists = RigVMFunctionReferenceNode(ExecuteContext=@RigUnit_BeginExecution.ExecutePin)
   > FN_DeformRig_FNC = RigVMFunctionReferenceNode(
    DeformRig_ Curve="(deformRig_l_upperarm_up_alpha_72_#22=False,deformRig_r_upperarm_up_alpha_73_#23=False,deformRig_l_upperarm_fr_alpha_74_#16=True,deformRig_r_upperarm_fr_alpha_75_CB51B80 ...(truncated)"
    DeformRig_Bool="(upperarm_1_#36=True,elbow_3_#42=true,wrist_7_#37=True,groin_8_#44=true,glute_17_#43=true,knee_9_40C408324CA261 ...(truncated)"
    DeformRig_Debug=(upperarm_debug_18_#36=false,wrist_debug_20_#37=False,lat_debug_26_#40=False,groinGlute_debug_27_#44=False)
    DeformRig_Value="(upperarm_up_18_#21=(Rotation=(X=0,Y=0,Z=0,W=1),Translation=(X=0,Y=0,Z=3),Scale3D=(X=1,Y=1,Z=1)),upperarm_fwd_fr_19_#14=(Rotation=(X=0,Y=0,Z=0,W=1),Translation=(X=-20,Y=10,Z=0),Scale3D=(X=1,Y=1,Z=1) ...(truncated)"
   )
   > Deform Tricep = RigVMFunctionReferenceNode()
   > Deform Scapula = RigVMFunctionReferenceNode()
   > Deform Stomach = RigVMFunctionReferenceNode()
   > Deform Lat Extra = RigVMFunctionReferenceNode()
   RigVMFunction_Sequence = &RigVMFunction_Sequence(ExecuteContext=())
  ]}
 SequenceAggregate_3_ContainedGraph {nodes [
   n1 = RigVMFunctionEntryNode()
   SequenceAggregate = &RigVMFunction_Sequence(ExecuteContext=@n1.ExecuteContext)
   SequenceAggregate_1 = &RigVMFunction_Sequence(ExecuteContext=@SequenceAggregate.B)
   RigVMFunctionReturnNode(A=@SequenceAggregate.A, B=@SequenceAggregate_1.A, C=@SequenceAggregate_1.B)
  ]}
]
hierarchy [
 root {loc=<0 0 0>, rot=<0 0 0 1>}
 pelvis {parent=root, loc=<0 -0.7321 92.4218>, rot=<0.0127 0.707 -0.0127 -0.707>}
 spine_01 {parent=pelvis, loc=<5.6002 -0.5128 0>, rot=<0 0 0.034 -0.9994>}
 spine_02 {parent=spine_01, loc=<8.0753 0.0139 0>, rot=<0 0 0.0238 -0.9997>}
 spine_03 {parent=spine_02, loc=<7.0393 0 0>, rot=<0 0 0.0149 -0.9999>}
 spine_04 {parent=spine_03, loc=<8.1024 0 0>, rot=<0 0 -0.0132 -0.9999>}
 spine_05 {parent=spine_04, loc=<10.0146 0 0>, rot=<0 0 -0.0015 -1>}
 clavicle_l {parent=spine_05, loc=<2.5856 -0.2866 -4.3777>, rot=<0.0465 -0.7289 -0.009 -0.6829>}
 upperarm_l {parent=clavicle_l, loc=<18.2264 0.1342 -0.001>, rot=<0.0253 -0.2693 0.0013 -0.9627>}
 lowerarm_l {parent=upperarm_l, loc=<29.0071 0.1686 -1.239>, rot=<0.0391 -0.0099 -0.2453 -0.9686>}
 hand_l {parent=lowerarm_l, loc=<25.8607 0.222 1.0274>, rot=<-0.7193 -0.0988 0.1193 -0.6773>}
 thumb_01_l {parent=hand_l, loc=<1.3118 -0.3191 -2.459>, rot=<0.6546 -0.4042 0.0664 -0.6354>}
 thumb_02_l {parent=thumb_01_l, loc=<5.9336 -0.0448 0.0401>, rot=<-0.0037 0.0213 0.298 -0.9543>}
 thumb_03_l {parent=thumb_02_l, loc=<2.2223 0.3682 -0.0907>, rot=<0.0118 -0.013 0.0066 -0.9998>}
 index_metacarpal_l {parent=hand_l, loc=<1.6535 0.0183 -2.96>, rot=<0.0158 -0.0124 -0.0173 -0.9996>}
 index_01_l {parent=index_metacarpal_l, loc=<8.3008 0.0357 -0.0524>, rot=<-0.0044 0.0485 0.0924 -0.9945>}
 index_02_l {parent=index_01_l, loc=<3.3587 -0.3004 0.0452>, rot=<-0.002 0.0027 0.3225 -0.9466>}
 index_03_l {parent=index_02_l, loc=<2.5578 -0.123 -0.0252>, rot=<-0.002 0.0067 0.2558 -0.9667>}
 middle_metacarpal_l {parent=hand_l, loc=<2.1856 -0.1725 -0.5984>, rot=<0.006 0.0168 -0.0329 -0.9993>}
 middle_01_l {parent=middle_metacarpal_l, loc=<7.3275 -0.2447 0.0129>, rot=<-0.0208 0.0042 0.0858 -0.9961>}
 middle_02_l {parent=middle_01_l, loc=<4.0123 0.1401 0.1175>, rot=<-0.0218 0.0258 0.3512 -0.9357>}
 middle_03_l {parent=middle_02_l, loc=<2.6077 0.0534 0.1373>, rot=<0.0082 0.0188 0.1708 -0.9851>}
 ring_metacarpal_l {parent=hand_l, loc=<2.2074 -0.3625 1.3638>, rot=<-0.0284 0.0722 -0.0458 -0.9959>}
 ring_01_l {parent=ring_metacarpal_l, loc=<7.0295 -0.6197 -0.1446>, rot=<-0.0489 -0.0252 0.1892 -0.9804>}
 ring_02_l {parent=ring_01_l, loc=<3.6174 0.4647 -0.1027>, rot=<0 0 0.2806 -0.9598>}
 ring_03_l {parent=ring_02_l, loc=<2.4152 0.114 0.0199>, rot=<0 0 0.1189 -0.9929>}
 pinky_metacarpal_l {parent=hand_l, loc=<1.7616 -0.9753 3.3814>, rot=<-0.0397 0.0814 -0.069 -0.9935>}
 pinky_01_l {parent=pinky_metacarpal_l, loc=<7.1709 -0.4831 -0.1094>, rot=<-0.0654 -0.0035 0.2242 -0.9723>}
 pinky_02_l {parent=pinky_01_l, loc=<2.8274 0.0893 -0.1043>, rot=<0.0037 0.0011 0.2854 -0.9584>}
 pinky_03_l {parent=pinky_02_l, loc=<2.2046 0.0307 -0.0078>, rot=<-0.0002 0.0098 0.1445 -0.9895>}
 lowerarm_twist_02_l {parent=lowerarm_l, loc=<9.6199 0 0>, rot=<0 0 0 -1>}
 lowerarm_twist_01_l {parent=lowerarm_l, loc=<17.8549 0 0>, rot=<0 0 0 -1>}
 upperarm_twist_01_l {parent=upperarm_l, loc=<7.9796 0.0843 -0.6195>, rot=<0 0 0 -1>}
 upperarm_twist_02_l {parent=upperarm_l, loc=<19.5059 0.0843 -0.6195>, rot=<0 0 0 -1>}
 deform_scap_base_l {parent=clavicle_l, loc=<9.5926 -17.7956 -5.0731>, rot=<0.0247 -0.0037 -0.0055 -0.9997>}
 deform_scap_up_l {parent=deform_scap_base_l, loc=<0 0 0>, rot=<0 0 0 1>}
 retarget_upperarm_l {parent=clavicle_l, loc=<18.2264 0.1342 -0.001>, rot=<0.0253 -0.2693 0.0013 -0.9627>}
 retarget_lowerarm_l {parent=retarget_upperarm_l, loc=<29.0071 0.1686 -1.239>, rot=<0.0391 -0.0099 -0.2453 -0.9686>}
 deform_elbow_in_base_l {parent=retarget_lowerarm_l, loc=<0.2112 5.645 0.8274>, rot=<0 0 0 -1>}
 deform_elbow_in_l {parent=deform_elbow_in_base_l, loc=<0 0 0>, rot=<0 0 0 1>}
 deform_elbow_out_base_l {parent=retarget_lowerarm_l, loc=<0 -7.5129 0>, rot=<0 0 0 -1>}
 deform_elbow_out_l {parent=deform_elbow_out_base_l, loc=<0 0 0>, rot=<0 0 0 1>}
 retarget_hand_l {parent=retarget_lowerarm_l, loc=<25.8607 0.222 1.0274>, rot=<-0.7193 -0.0988 0.1193 -0.6773>}
 deform_wrist_dn_base_l {parent=retarget_hand_l, loc=<0 -3.4173 0>, rot=<0 0 0 -1>}
 deform_wrist_dn_l {parent=deform_wrist_dn_base_l, loc=<0 0 0>, rot=<0 0 0 1>}
 deform_wrist_up_base_l {parent=retarget_hand_l, loc=<0 2.7805 0>, rot=<0 0 0 -1>}
 deform_wrist_up_l {parent=deform_wrist_up_base_l, loc=<0 0 0>, rot=<0 0 0 1>}
 retarget_lowerarm_twist_02_l {parent=retarget_lowerarm_l, loc=<9.6199 0 0>, rot=<0 0 0 -1>}
 retarget_lowerarm_twist_01_l {parent=retarget_lowerarm_l, loc=<17.8549 0 0>, rot=<0 0 0 -1>}
 retarget_hand_split_l {parent=retarget_lowerarm_l, loc=<25.8607 0.222 1.0274>, rot=<-0.7193 -0.0988 0.1193 -0.6773>}
 retarget_upperarm_twist_01_l {parent=retarget_upperarm_l, loc=<7.9796 0.0843 -0.6195>, rot=<0 0 0 -1>}
 deform_upperarm_bk_base_l {parent=retarget_upperarm_twist_01_l, loc=<-6.8983 -13.3826 -1.3258>, rot=<0 0 0 -1>}
 deform_upperarm_bk_l {parent=deform_upperarm_bk_base_l, loc=<0 0 0>, rot=<0 0 0 1>}
 deform_upperarm_fr_base_l {parent=retarget_upperarm_twist_01_l, loc=<-8.3487 10.7117 -2.1455>, rot=<0 0 0 -1>}
 deform_upperarm_fr_l {parent=deform_upperarm_fr_base_l, loc=<0 0 0>, rot=<0 0 0 1>}
 deform_upperarm_up_base_l {parent=retarget_upperarm_twist_01_l, loc=<-11.2012 -1.2234 8.7784>, rot=<0 0 0 -1>}
 deform_upperarm_up_l {parent=deform_upperarm_up_base_l, loc=<0 0 0>, rot=<0 0 0 1>}
 retarget_upperarm_twist_02_l {parent=retarget_upperarm_l, loc=<19.5059 0.0843 -0.6195>, rot=<0 0 0 -1>}
 deform_bicep_base_l {parent=retarget_upperarm_twist_02_l, loc=<0 8.0494 0>, rot=<0 0 0 1>}
 deform_bicep_l {parent=deform_bicep_base_l, loc=<0 0 0>, rot=<0 0 0 1>}
 deform_tricep_base_l {parent=retarget_upperarm_twist_02_l, loc=<-2.4201 -8.0891 0>, rot=<0 0 -0.1207 -0.9927>}
 deform_tricep_l {parent=deform_tricep_base_l, loc=<0 0 0>, rot=<0 0 0 1>}
 neck_01 {parent=spine_05, loc=<10.815 1.9196 0>, rot=<0 0 -0.1864 -0.9825>}
 neck_02 {parent=neck_01, loc=<4.7347 0.3045 0>, rot=<0 0 0.0363 -0.9993>}
 head {parent=neck_02, loc=<5.7632 0.5552 0>, rot=<0 0 0.115 -0.9934>}
 attach_cape {parent=spine_05, loc=<9.1509 -13.3783 0>, rot=<0.0616 -0.7044 -0.0616 -0.7044>}
 attach_backpack {parent=spine_05, loc=<-0.123 -15.1423 0>, rot=<-0.0093 -0.707 0.0093 -0.707>}
 clavicle_r {parent=spine_05, loc=<2.5855 -0.2866 4.3777>, rot=<0.7289 0.0465 -0.6829 0.009>}
 upperarm_r {parent=clavicle_r, loc=<-18.2264 -0.1342 0.001>, rot=<0.0253 -0.2693 0.0013 -0.9627>}
 lowerarm_r {parent=upperarm_r, loc=<-29.0071 -0.1686 1.239>, rot=<0.0391 -0.0099 -0.2453 -0.9686>}
 hand_r {parent=lowerarm_r, loc=<-25.8608 -0.222 -1.0274>, rot=<-0.7193 -0.0988 0.1193 -0.6773>}
 thumb_01_r {parent=hand_r, loc=<-1.3118 0.3191 2.459>, rot=<0.6546 -0.4042 0.0664 -0.6354>}
 thumb_02_r {parent=thumb_01_r, loc=<-5.9336 0.0448 -0.0401>, rot=<-0.0037 0.0213 0.298 -0.9543>}
 thumb_03_r {parent=thumb_02_r, loc=<-2.2223 -0.3682 0.0907>, rot=<0.0118 -0.013 0.0066 -0.9998>}
 index_metacarpal_r {parent=hand_r, loc=<-1.6535 -0.0183 2.96>, rot=<0.0158 -0.0124 -0.0173 -0.9996>}
 index_01_r {parent=index_metacarpal_r, loc=<-8.3008 -0.0357 0.0524>, rot=<-0.0044 0.0485 0.0924 -0.9945>}
 index_02_r {parent=index_01_r, loc=<-3.3587 0.3004 -0.0452>, rot=<-0.002 0.0027 0.3225 -0.9466>}
 index_03_r {parent=index_02_r, loc=<-2.5578 0.123 0.0252>, rot=<-0.002 0.0067 0.2558 -0.9667>}
 middle_metacarpal_r {parent=hand_r, loc=<-2.1856 0.1725 0.5984>, rot=<0.006 0.0168 -0.0329 -0.9993>}
 middle_01_r {parent=middle_metacarpal_r, loc=<-7.3275 0.2447 -0.0129>, rot=<-0.0208 0.0042 0.0858 -0.9961>}
 middle_02_r {parent=middle_01_r, loc=<-4.0123 -0.1401 -0.1175>, rot=<-0.0218 0.0258 0.3512 -0.9357>}
 middle_03_r {parent=middle_02_r, loc=<-2.6077 -0.0534 -0.1373>, rot=<0.0082 0.0188 0.1708 -0.9851>}
 ring_metacarpal_r {parent=hand_r, loc=<-2.2074 0.3625 -1.3638>, rot=<-0.0284 0.0722 -0.0458 -0.9959>}
 ring_01_r {parent=ring_metacarpal_r, loc=<-7.0295 0.6197 0.1446>, rot=<-0.0489 -0.0252 0.1892 -0.9804>}
 ring_02_r {parent=ring_01_r, loc=<-3.6174 -0.4647 0.1027>, rot=<0 0 0.2806 -0.9598>}
 ring_03_r {parent=ring_02_r, loc=<-2.4152 -0.114 -0.0199>, rot=<0 0 0.1189 -0.9929>}
 pinky_metacarpal_r {parent=hand_r, loc=<-1.7616 0.9753 -3.3814>, rot=<-0.0397 0.0814 -0.069 -0.9935>}
 pinky_01_r {parent=pinky_metacarpal_r, loc=<-7.1709 0.4831 0.1094>, rot=<-0.0654 -0.0035 0.2242 -0.9723>}
 pinky_02_r {parent=pinky_01_r, loc=<-2.8274 -0.0893 0.1043>, rot=<0.0037 0.0011 0.2854 -0.9584>}
 pinky_03_r {parent=pinky_02_r, loc=<-2.2046 -0.0307 0.0078>, rot=<-0.0002 0.0098 0.1445 -0.9895>}
 lowerarm_twist_02_r {parent=lowerarm_r, loc=<-9.6199 0 0>, rot=<0 0 0 -1>}
 lowerarm_twist_01_r {parent=lowerarm_r, loc=<-17.8549 0 0>, rot=<0 0 0 -1>}
 upperarm_twist_01_r {parent=upperarm_r, loc=<-7.9796 -0.0843 0.6195>, rot=<0 0 0 -1>}
 upperarm_twist_02_r {parent=upperarm_r, loc=<-19.5059 -0.0843 0.6195>, rot=<0 0 0 -1>}
 deform_scap_base_r {parent=clavicle_r, loc=<-9.5926 17.7956 5.0732>, rot=<0.0247 -0.0037 -0.0055 -0.9997>}
 deform_scap_up_r {parent=deform_scap_base_r, loc=<0 0 0>, rot=<0 0 0 1>}
 retarget_upperarm_r {parent=clavicle_r, loc=<-18.2264 -0.1342 0.001>, rot=<0.0253 -0.2693 0.0013 -0.9627>}
 retarget_lowerarm_r {parent=retarget_upperarm_r, loc=<-29.0072 -0.1686 1.239>, rot=<0.0391 -0.0099 -0.2453 -0.9686>}
 deform_elbow_in_base_r {parent=retarget_lowerarm_r, loc=<-0.2114 -5.645 -0.827>, rot=<0 0 0 -1>}
 deform_elbow_in_r {parent=deform_elbow_in_base_r, loc=<0 0 0>, rot=<0 0 0 1>}
 deform_elbow_out_base_r {parent=retarget_lowerarm_r, loc=<-0.0001 7.5129 0.0002>, rot=<0 0 0 -1>}
 deform_elbow_out_r {parent=deform_elbow_out_base_r, loc=<0 0 0>, rot=<0 0 0 1>}
 retarget_hand_r {parent=retarget_lowerarm_r, loc=<-25.8608 -0.222 -1.0274>, rot=<-0.7193 -0.0988 0.1193 -0.6773>}
 deform_wrist_dn_base_r {parent=retarget_hand_r, loc=<0.0001 3.4173 0>, rot=<0 0 0 -1>}
 deform_wrist_dn_r {parent=deform_wrist_dn_base_r, loc=<0 0 0>, rot=<0 0 0 -1>}
 deform_wrist_up_base_r {parent=retarget_hand_r, loc=<-0.0001 -2.7804 0>, rot=<0 0 0 -1>}
 deform_wrist_up_r {parent=deform_wrist_up_base_r, loc=<0 0 0>, rot=<0 0 0 -1>}
 retarget_lowerarm_twist_02_r {parent=retarget_lowerarm_r, loc=<-9.6199 0 0>, rot=<0 0 0 -1>}
 retarget_lowerarm_twist_01_r {parent=retarget_lowerarm_r, loc=<-17.8549 0 0>, rot=<0 0 0 -1>}
 retarget_hand_split_r {parent=retarget_lowerarm_r, loc=<-25.8608 -0.222 -1.0274>, rot=<-0.7193 -0.0988 0.1193 -0.6773>}
 retarget_upperarm_twist_01_r {parent=retarget_upperarm_r, loc=<-7.9796 -0.0843 0.6195>, rot=<0 0 0 -1>}
 deform_upperarm_bk_base_r {parent=retarget_upperarm_twist_01_r, loc=<6.8985 13.3826 1.3257>, rot=<0 0 0 -1>}
 deform_upperarm_bk_r {parent=deform_upperarm_bk_base_r, loc=<0 0 0>, rot=<0 0 0 1>}
 deform_upperarm_fr_base_r {parent=retarget_upperarm_twist_01_r, loc=<8.3485 -10.7117 2.1456>, rot=<0 0 0 -1>}
 deform_upperarm_fr_r {parent=deform_upperarm_fr_base_r, loc=<0 0 0>, rot=<0 0 0 1>}
 deform_upperarm_up_base_r {parent=retarget_upperarm_twist_01_r, loc=<11.2015 1.2234 -8.7788>, rot=<0 0 0 -1>}
 deform_upperarm_up_r {parent=deform_upperarm_up_base_r, loc=<0 0 0>, rot=<0 0 0 1>}
 retarget_upperarm_twist_02_r {parent=retarget_upperarm_r, loc=<-19.5059 -0.0843 0.6195>, rot=<0 0 0 -1>}
 deform_bicep_base_r {parent=retarget_upperarm_twist_02_r, loc=<0.0002 -8.0494 -0.0003>, rot=<0 0 0 1>}
 deform_bicep_r {parent=deform_bicep_base_r, loc=<0 0 0>, rot=<0 0 0 1>}
 deform_tricep_base_r {parent=retarget_upperarm_twist_02_r, loc=<2.4201 8.0892 0>, rot=<0 0 -0.1207 -0.9927>}
 deform_tricep_r {parent=deform_tricep_base_r, loc=<0 0 0>, rot=<0 0 0 -1>}
 deform_pec_aim_base_l {parent=spine_04, loc=<5.0232 13.6719 -2.6508>, rot=<0.0672 -0.5201 0.3228 -0.7879>}
 deform_pec_aim_l {parent=deform_pec_aim_base_l, loc=<0 0 0>, rot=<0 0 0 1>}
 deform_pec_base_l {parent=deform_pec_aim_l, loc=<16.3279 0 0>, rot=<0 0 0 1>}
 deform_pec_l {parent=deform_pec_base_l, loc=<0 0 0>, rot=<0 0 0 1>}
 deform_pec_aim_base_r {parent=spine_04, loc=<5.023 13.6718 2.6508>, rot=<-0.5201 -0.0672 0.7879 0.3228>}
 deform_pec_aim_r {parent=deform_pec_aim_base_r, loc=<0 0 0>, rot=<0 0 0 1>}
 deform_pec_base_r {parent=deform_pec_aim_r, loc=<-16.3279 0 0>, rot=<0 0 0 -1>}
 deform_pec_r {parent=deform_pec_base_r, loc=<0 0 0.0002>, rot=<0 0 0 -1>}
 deform_lat_l_base {parent=spine_03, loc=<2.5566 -10.3716 -17.6918>, rot=<0 0 0 1>}
 deform_lat_l {parent=deform_lat_l_base, loc=<0 0 0>, rot=<0 0 0 1>}
 deform_lat_r_base {parent=spine_03, loc=<2.5568 -10.3716 17.6918>, rot=<0 0 1 0>}
 deform_lat_r {parent=deform_lat_r_base, loc=<0 0 0>, rot=<0 0 0 -1>}
 deform_stomach_base {parent=spine_02, loc=<-6.3614 14.8733 0>, rot=<0 0 -0.0238 -0.9997>}
 deform_stomach {parent=deform_stomach_base, loc=<0 0 0>, rot=<0 0 0 1>}
 thigh_l {parent=pelvis, loc=<-1.8716 -0.0109 -11.9936>, rot=<0.0012 0.0572 0.0155 -0.9982>}
 thigh_twist_01_l {parent=thigh_l, loc=<-21.6303 0 0>, rot=<0.0013 0 0 -1>}
 thigh_r {parent=pelvis, loc=<-1.8714 -0.0109 11.9936>, rot=<0.0572 -0.0012 0.9982 0.0155>}
 thigh_twist_01_r {parent=thigh_r, loc=<21.6303 0 0>, rot=<0.0013 0 0 -1>}
 deform_glute_base_l {parent=pelvis, loc=<-10.1877 -11.6217 -12.899>, rot=<0.0275 0.0501 -0.4472 -0.8926>}
 deform_glute_l {parent=deform_glute_base_l, loc=<0 0 0>, rot=<0 0 0 1>}
 deform_glute_base_r {parent=pelvis, loc=<-10.1877 -11.6218 12.899>, rot=<0.0501 -0.0275 0.8926 -0.4472>}
 deform_glute_r {parent=deform_glute_base_r, loc=<0 0 0>, rot=<0 0 0 1>}
 retarget_thigh_l {parent=pelvis, loc=<-5.0484 0.0888 -10.4778>, rot=<0.0014 0.0693 0.0154 -0.9975>}
 retarget_calf_l {parent=retarget_thigh_l, loc=<-38.9323 -0.2246 0.0926>, rot=<0.0006 -0.0161 -0.0391 -0.9991>}
 deform_knee_in_base_l {parent=retarget_calf_l, loc=<0.7814 -5.0151 -0.0712>, rot=<-0.0546 -0.0067 -0.0133 -0.9984>}
 deform_knee_in_l {parent=deform_knee_in_base_l, loc=<0 0 0>, rot=<0 0 0 1>}
 deform_knee_out_base_l {parent=retarget_calf_l, loc=<-0.8774 5.0408 0.6018>, rot=<-0.001 -0.0121 0 -0.9999>}
 deform_knee_out_l {parent=deform_knee_out_base_l, loc=<0 0 0>, rot=<0 0 0 1>}
 retarget_foot_l {parent=retarget_calf_l, loc=<-42.7715 1.5393 0.605>, rot=<0.0975 -0.0474 0.0317 -0.9936>}
 retarget_ball_l {parent=retarget_foot_l, loc=<-4.598 11.6171 0.3696>, rot=<0.0001 -0.0001 -0.7063 -0.7079>}
 retarget_calf_twist_01_l {parent=retarget_calf_l, loc=<-42.7708 1.2186 0.5985>, rot=<0.0538 -0.0054 0.0126 -0.9985>}
 retarget_calf_twist_02_l {parent=retarget_calf_l, loc=<-21.8575 0.6785 0.4006>, rot=<0.0538 -0.0054 0.0126 -0.9985>}
 retarget_thigh_twist_01_l {parent=retarget_thigh_l, loc=<0 0 0>, rot=<0.0544 -0.0024 0.0005 -0.9985>}
 deform_groin_base_l {parent=retarget_thigh_twist_01_l, loc=<-4.2509 5.6127 -0.7514>, rot=<-0.0544 -0.0098 -0.0011 -0.9985>}
 deform_groin_l {parent=deform_groin_base_l, loc=<0 0 0>, rot=<0 0 0 1>}
 retarget_thigh_r {parent=pelvis, loc=<-5.0484 0.0888 10.4778>, rot=<0.0693 -0.0014 0.9975 0.0154>}
 retarget_calf_r {parent=retarget_thigh_r, loc=<38.9323 0.2247 -0.0926>, rot=<0.0006 -0.0161 -0.0391 -0.9991>}
 deform_knee_in_base_r {parent=retarget_calf_r, loc=<-0.7813 5.0151 0.0712>, rot=<-0.001 -0.0121 0 -0.9999>}
 deform_knee_in_r {parent=deform_knee_in_base_r, loc=<0 0 0>, rot=<0 0 0 1>}
 deform_knee_out_base_r {parent=retarget_calf_r, loc=<0.8774 -5.0408 -0.6018>, rot=<-0.001 -0.0121 0 -0.9999>}
 deform_knee_out_r {parent=deform_knee_out_base_r, loc=<0 0 0>, rot=<0 0 0 1>}
 retarget_foot_r {parent=retarget_calf_r, loc=<42.7715 -1.5393 -0.605>, rot=<0.0975 -0.0474 0.0317 -0.9936>}
 retarget_ball_r {parent=retarget_foot_r, loc=<4.598 -11.6171 -0.3696>, rot=<0.0001 -0.0001 -0.7063 -0.7079>}
 retarget_calf_twist_01_r {parent=retarget_calf_r, loc=<42.7709 -1.2186 -0.5985>, rot=<0.0538 -0.0054 0.0126 -0.9985>}
 retarget_calf_twist_02_r {parent=retarget_calf_r, loc=<21.8575 -0.6785 -0.4005>, rot=<0.0538 -0.0054 0.0126 -0.9985>}
 retarget_thigh_twist_01_r {parent=retarget_thigh_r, loc=<0 0 0>, rot=<0.0544 -0.0024 0.0005 -0.9985>}
 deform_groin_base_r {parent=retarget_thigh_twist_01_r, loc=<4.251 -5.6127 0.7515>, rot=<-0.0544 -0.0098 -0.0011 -0.9985>}
 deform_groin_r {parent=deform_groin_base_r, loc=<0 0 0>, rot=<0 0 0 -1>}
]
variables [
 DisableRuntimeArmTwists {type=bool, default=False}
 LeftArmCloseToSternum {type=double, default=0}
 RightArmCloseToSternum {type=double, default=0}
]
