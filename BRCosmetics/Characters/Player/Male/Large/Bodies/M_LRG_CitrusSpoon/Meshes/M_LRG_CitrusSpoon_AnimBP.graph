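# Animation Blueprint for the CitrusSpoon large male body, adapting the parent mesh's pose when a shoe mesh is present and copying it otherwise.
# The adapted pose layers copied arms, neck and thighs over an IK-retargeted base, then runs a Control Rig.
# Arm twists are disabled during emotes or without a pawn owner.
# Otherwise, updates rebuild an arm retarget profile with curve-driven pole-vector offsets from forearm-to-spine distance, though those alignment entries are disabled.
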AnimBlueprint /BRCosmetics/Characters/Player/Male/Large/Bodies/M_LRG_CitrusSpoon/Meshes/M_LRG_CitrusSpoon_AnimBP
target_skeleton=&/M_LRG_CitrusSpoon_Skeleton
graphs [
 AnimGraph {nodes [
   n1 = Layered blend per bone(
    BasePose=Retarget Pose From Mesh!(CustomRetargetProfile:Retarget Profile Structure=(RetargetOpProfiles=,bApplyTargetRetargetPose=False,TargetRetargetPoseName="",bApplySourceRetargetPose=False,SourceRetargetPoseName="",bForceAllIKOff=False,bApplyChainSettings=False,ChainSettings=(),bApplyRootSettings=False,RootSettings=(RotationAlpha=1,TranslationAlpha=1,BlendToSource=0,BlendToSourceWeights=(X=1,Y=1,Z=1),ScaleHorizontal=1,ScaleVertical=1,TranslationOffset=(X=0,Y=0,Z=0),RotationOffset=(Pitch=0,Yaw=0,Roll=0),AffectIKHorizontal=1,AffectIKVertical=0),bApplyGlobalSettings=False,GlobalSettings=(bEnableRoot=True,bEnableFK=True,bEnableIK=True,bEnablePost=True,bCopyBasePose=False,CopyBasePoseRoot="",SourceScaleFactor=1,bWarping=False,DirectionSource=Goals,ForwardDirection=Y,DirectionChain="",WarpForwards=1,SidewaysOffset=0,WarpSplay=1))) -> (Pose:Pose Link Structure) {
     is=title AnimGraphNode_RetargetPoseFromMesh
     iKRetargeterAsset=&/RTG_M_LRG_CitrusSpoon
    }
    BlendPoses_0=Copy Pose From Mesh() {bUseAttachedParent=true}
   ) {layerSetup [{branchFilters [{boneName=upperarm_l}, {boneName=upperarm_r}, {boneName=neck_01}, {boneName=thigh_l}, {boneName=thigh_r}]}]}
   n2 = Control Rig!(Source=@n1, Alpha, DisableRuntimeArmTwists:bool=Get DisableArmCR() -> (DisableArmCR:bool)) -> (Pose) {
    bTransferInputCurves=false
    alphaBoolBlend {blendOption=Linear}
    controlRigAssetReference {blueprintRigClass=&/M_LRG_CitrusSpoon_CtrlRig.M_LRG_CitrusSpoon_CtrlRig_C}
   }
   Output Pose(Result=Blend Poses by bool(BlendPose_0=Copy Pose From Mesh() {bUseAttachedParent=true}, BlendPose_1=@n2, bActiveValue=Get Shoes_Active() -> (Shoes_Active:bool)) {blendType=Linear}) {layerGroup=None}
  ]}
 EventGraph {nodes [
   Event BlueprintUpdateAnimation()
   > n1 = Sequence()
   n2 = Get bIsPlayingEmote()
   > n3 = Branch(Condition=@n2)
   > Set DisableArmCR(DisableArmCR:bool=true) -> (Output_Get:bool)
   n4 = Is Valid(exec=@n3.else, InputObject=TryGetPawnOwner())
   Set DisableArmCR(execute=@n4.Is Not Valid, DisableArmCR=true)
   Set DisableArmCR(execute=@n4.Is Valid, DisableArmCR=false)
   > n5 = Sequence!(execute) -> (then_0, then_1, then_2:Exec)
   n6 = Sequence!(execute=@n5.then_2) -> (then_0, then_1, then_2, then_3:Exec, then_4:Exec)
   > Clear(TargetArray:Array of Retarget Op Profile Structures=Get Retarget Op Profiles() -> (Retarget Op Profiles:Array of Retarget Op Profile Structures))
   n7 = MakeInstancedStruct!(
    execute:Exec=@n6.then_1
    Value:Retarget IKChain Settings Structure=Make Retarget IKChain Settings!(
     TargetChainName:Name=LeftArm
     EnableIK:bool=True
     BlendToSource:double=1
     BlendToSourceTranslation:double=1
     BlendToSourceRotation:double=0
     BlendToSourceWeights:Vector="1,1,1"
     ApplyPelvisOffsetToSourceGoals:bool=False
     StaticOffset:Vector="0,0,0"
     StaticLocalOffset:Vector="0,0,0"
     StaticRotationOffset:Rotator="0,0,0"
     ScaleVertical:double=1
     Extension:double=1
    ) -> (RetargetIKChainSettings:Retarget IKChain Settings Structure) {is=title K2Node_MakeStruct}
   ) -> (then:Exec, ReturnValue:Instanced Struct Structure) {is=title K2Node_InstancedStruct}
   > Add(
    TargetArray:Array of Retarget Op Profile Structures=Get Retarget Op Profiles()
    NewItem:"Retarget Op Profile Structure (by ref)"=Make Retarget Op Profile!(OpToApplySettingsTo:Name=None, SettingsToApply:Instanced Struct Structure=@n7) -> (RetargetOpProfile:Retarget Op Profile Structure) {is=title K2Node_MakeStruct}
   ) {is=title K2Node_CallArrayFunction}
   n8 = Make Retarget Pole Vector Settings!(
    TargetChainName:Name=LeftArm
    bEnabled:bool=False
    AlignAlpha:double=1
    StaticAngularOffset:double=Get lowerarm_l_polevector_tweak() -> (lowerarm_l_polevector_tweak:double)
    MaintainOffset:bool=False
   ) -> (RetargetPoleVectorSettings:Retarget Pole Vector Settings Structure) {is=title K2Node_MakeStruct}
   n9 = Make Retarget Pole Vector Settings(
    TargetChainName=RightArm
    bEnabled=False
    AlignAlpha=1
    StaticAngularOffset=Get lowerarm_r_polevector_tweak() -> (lowerarm_r_polevector_tweak:double)
    MaintainOffset=False
   )
   n10 = MakeInstancedStruct(
    execute=@n6.then_3
    Value:Pole Vector Alignment Settings Structure=Make Pole Vector Alignment Settings!(
     IKRigAsset:IKRig Definition Object Reference
     ChainsToAlign:Array of Retarget Pole Vector Settings Structures=Make Array!("[0]":Retarget Pole Vector Settings Structure=@n8, "[1]":Retarget Pole Vector Settings Structure=@n9) -> (Array:Array of Retarget Pole Vector Settings Structures)
     Alpha:double=1
     bEnabled:bool=True
     LODThreshold:int=-1
    ) -> (IKRetargetAlignPoleVectorOpSettings:Pole Vector Alignment Settings Structure) {is=title K2Node_MakeStruct}
   )
   > Add(TargetArray=Get Retarget Op Profiles(), NewItem=Make Retarget Op Profile(OpToApplySettingsTo=None, SettingsToApply=@n10))
   n11 = MakeInstancedStruct(
    execute=@n6.then_2
    Value:Retarget IKChain Settings Structure=Make Retarget IKChain Settings(
     TargetChainName=RightArm
     EnableIK=True
     BlendToSource=1
     BlendToSourceTranslation=1
     BlendToSourceRotation=0
     BlendToSourceWeights="1,1,1"
     ApplyPelvisOffsetToSourceGoals=False
     StaticOffset="0,0,0"
     StaticLocalOffset="0,0,0"
     StaticRotationOffset="0,0,0"
     ScaleVertical=1
     Extension=1
    )
   )
   > Add(TargetArray=Get Retarget Op Profiles(), NewItem=Make Retarget Op Profile(OpToApplySettingsTo=None, SettingsToApply=@n11))
   Set members in Retarget Profile!(
    execute:Exec=@n6.then_4
    StructRef:"Retarget Profile Structure (by ref)"=Get Custom Retarget Profile() -> (Custom Retarget Profile:Retarget Profile Structure)
    RetargetOpProfiles:Array of Retarget Op Profile Structures=Get Retarget Op Profiles()
   ) -> (then:Exec, StructOut:"Retarget Profile Structure (by ref)") {is=title K2Node_SetFieldsInStruct}
   n12 = GetPartSkeletalMeshComponent()
   n13 = GetBoneTransform!(self:Skinned Mesh Component Object Reference=@n12, InBoneName:Name=spine_04, TransformSpace:ERelativeTransformSpace Enum=RTS_World) -> (ReturnValue_Location:Vector, ReturnValue_Rotation:Rotator, ReturnValue_Scale:Vector)
   n14 = DistanceBetweenTwoVectors!(execute=@n5.then_1, vector1:Vector=@n13, Vector2:Vector=GetBoneTransform(self=@n12, InBoneName=lowerarm_r, TransformSpace=RTS_World)) -> (then, Distance:double)
   n15 = "Clamp (Float)"(Value=SafeDivide(A=20, B=@n14), Min=0, Max=1)
   n16 = Get Curve() -> (Curve:Runtime Float Curve Structure)
   > Set lowerarm_r_polevector_tweak(lowerarm_r_polevector_tweak:double=float * float(
     A=GetRuntimeFloatCurveValue!(Curve:"Runtime Float Curve Structure (by ref)"=@n16, InTime:float=@n15, InDefaultValue:float=0) -> (ReturnValue:float)
     B=32
    ) {is=title K2Node_PromotableOperator}) -> (Output_Get:double)
   n17 = DistanceBetweenTwoVectors(execute=@n5.then_0, vector1=GetBoneTransform(self=@n12, InBoneName=lowerarm_l, TransformSpace=RTS_World), Vector2=@n13)
   n18 = "Clamp (Float)"(Value=SafeDivide(A=20, B=@n17), Min=0, Max=1)
   n19 = Get Curve()
   > Set lowerarm_l_polevector_tweak(lowerarm_l_polevector_tweak:double=float * float(A=GetRuntimeFloatCurveValue(Curve=@n19, InTime=@n18, InDefaultValue=0), B=-32)) -> (Output_Get:double)
   n20 = Get!(execute:Exec=@n1.then_1) -> (then:Exec, else:Exec, ShoeMesh:Skeletal Mesh Component Budgeted Object Reference)
   > n21 = Is Valid(InputObject=Get SkeletalMesh!(self:Skinned Mesh Component Object Reference=@n20) -> (SkeletalMesh:Skeletal Mesh Object Reference))
   Set Shoes_Active(execute=[@n21.Is Not Valid, @n20.else], Shoes_Active:bool=false) -> (Output_Get:bool)
   Set Shoes_Active(execute=@n21.Is Valid, Shoes_Active=true)
   n22 = Event OnCharacterCustomizationLoaded()
   > Set ShoeMesh(ShoeMesh:Skeletal Mesh Component Budgeted Object Reference=Get Array Item with EFort Custom Part Type(Array=@n22.CustomMeshParts, PartType=ExtraPart)) -> (Output_Get:Skeletal Mesh Component Budgeted Object Reference)
   OR Boolean(A=@n2, B=Get bIsInFrontEnd())
  ]}
]
variables [
 Chain Settings IK {
  type=RetargetIKChainSettings
  default {
   applyPelvisOffsetToSourceGoals=true
   blendToSource=1
   blendToSourceRotation=0
   blendToSourceTranslation=1
   enableIK=true
   extension=1
   scaleVertical=1
   targetChainName=None
   blendToSourceWeights=<1 1 1>
   staticLocalOffset=<0 0 0>
   staticOffset=<0 0 0>
   staticRotationOffset=<rot 0 0 0>
  }
 }
 Curve {
  type=RuntimeFloatCurve
  default {
   externalCurve=&/M_LRG_CitrusSpoon_PV_crv
   editorCurveData {
    defaultValue=3.4028234663852886e+38
    postInfinityExtrap=RCCE_Constant
    preInfinityExtrap=RCCE_Constant
    keys [
     {
      arriveTangent=0
      arriveTangentWeight=0
      interpMode=RCIM_Cubic
      leaveTangent=0
      leaveTangentWeight=0
      tangentMode=RCTM_Auto
      tangentWeightMode=RCTWM_WeightedNone
      time=0
      value=0
     }
     {
      arriveTangent=1
      arriveTangentWeight=0
      interpMode=RCIM_Cubic
      leaveTangent=1
      leaveTangentWeight=0
      tangentMode=RCTM_Auto
      tangentWeightMode=RCTWM_WeightedNone
      time=0.8037
      value=0.2488
     }
     {
      arriveTangent=0
      arriveTangentWeight=0
      interpMode=RCIM_Linear
      leaveTangent=0
      leaveTangentWeight=0
      tangentMode=RCTM_Auto
      tangentWeightMode=RCTWM_WeightedNone
      time=1
      value=1
     }
    ]
   }
  }
 }
 Custom Retarget Profile {
  type=RetargetProfile
  default {
   bApplyChainSettings=true
   bApplyGlobalSettings=false
   bApplyRootSettings=false
   bApplySourceRetargetPose=false
   bApplyTargetRetargetPose=false
   bForceAllIKOff=false
   sourceRetargetPoseName=None
   targetRetargetPoseName=None
   chainSettings {
    leftArm {
     fK {
      enableFK=true
      poleVectorMaintainOffset=false
      poleVectorMatching=0
      poleVectorOffset=0
      rotationAlpha=1
      rotationMode=Interpolated
      translationAlpha=1
      translationMode=None
     }
     iK {
      bAffectedByIKWarping=true
      blendToSource=1
      blendToSourceRotation=0
      blendToSourceTranslation=1
      enableIK=true
      extension=1
      scaleVertical=1
      blendToSourceWeights=<1 1 1>
      staticLocalOffset=<0 0 0>
      staticOffset=<0 0 0>
      staticRotationOffset=<rot 0 0 0>
     }
     speedPlanting {enableSpeedPlanting=false, speedCurveName=None, speedThreshold=15, unplantCriticalDamping=1, unplantStiffness=250}
    }
    rightArm {
     fK {
      enableFK=true
      poleVectorMaintainOffset=false
      poleVectorMatching=0
      poleVectorOffset=0
      rotationAlpha=1
      rotationMode=Interpolated
      translationAlpha=1
      translationMode=None
     }
     iK {
      bAffectedByIKWarping=true
      blendToSource=1
      blendToSourceRotation=0
      blendToSourceTranslation=1
      enableIK=true
      extension=1
      scaleVertical=1
      blendToSourceWeights=<1 1 1>
      staticLocalOffset=<0 0 0>
      staticOffset=<0 0 0>
      staticRotationOffset=<rot 0 0 0>
     }
     speedPlanting {enableSpeedPlanting=false, speedCurveName=None, speedThreshold=15, unplantCriticalDamping=1, unplantStiffness=250}
    }
   }
   globalSettings {
    bCopyBasePose=false
    bEnableFK=true
    bEnableIK=true
    bEnablePost=true
    bEnableRoot=true
    bWarping=false
    copyBasePoseRoot=None
    directionChain=None
    directionSource=Goals
    forwardDirection=Y
    sidewaysOffset=0
    sourceScaleFactor=1
    warpForwards=1
    warpSplay=1
   }
   retargetOpProfiles []
   rootSettings {
    affectIKHorizontal=1
    affectIKVertical=0
    blendToSource=0
    rotationAlpha=1
    scaleHorizontal=1
    scaleVertical=1
    translationAlpha=1
    blendToSourceWeights=<1 1 1>
    rotationOffset=<rot 0 0 0>
    translationOffset=<0 0 0>
   }
  }
 }
 DisableArmCR {type=boolean}
 lowerarm_l_polevector_tweak {type=number}
 lowerarm_r_polevector_tweak {type=number}
 Retarget Op Profiles {type="array<RetargetOpProfile>", default []}
 ShoeMesh {type=SkeletalMeshComponentBudgeted}
 Shoes_Active {type=boolean}
]
