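# Physics asset for the face accessory part of the large male CitrusSpoon player cosmetic, covering a dynamic hair strand and the head.
# Five small capsule bodies on dyn_hair_01 through dyn_hair_05 form a serial chain, with the root constrained to a large head capsule.
# Every joint shares the same narrow swing and twist limits, so each link can bend only slightly relative to its parent.
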PhysicsAsset /BRCosmetics/Characters/Player/Male/Large/Bodies/M_LRG_CitrusSpoon/Meshes/Parts/M_LRG_CitrusSpoon_FaceAcc_Physics
bodies {
 dyn_hair_01 [{shapeType=Capsule, length=0.505, radius=0.505, shapeName=dyn_hair_01_capsule, rotation=<rot 0 -90 0>}]
 dyn_hair_02 [{
   shapeType=Capsule
   length=0.505
   radius=0.505
   shapeName=dyn_hair_02_capsule
   center=<0.128 -0.0155 0.0289>
   rotation=<rot 0 87.6015 -81.2532>
  }]
 dyn_hair_03 [{
   shapeType=Capsule
   length=0.505
   radius=0.505
   shapeName=dyn_hair_03_capsule
   center=<0.3117 0.0044 0.0107>
   rotation=<rot 0 87.549 -89.2127>
  }]
 dyn_hair_04 [{
   shapeType=Capsule
   length=0.505
   radius=0.505
   shapeName=dyn_hair_04_capsule
   center=<0.4444 -0.0003 -0.0022>
   rotation=<rot 0 -91.0162 -89.6691>
  }]
 dyn_hair_05 [{
   shapeType=Capsule
   length=0.505
   radius=0.505
   shapeName=dyn_hair_05_capsule
   center=<0.7021 -0.2166 0.0107>
   rotation=<rot 0 69.751 -89.4345>
  }]
 head [{shapeType=Capsule, length=12.1875, radius=8.1875, shapeName=head_capsule, center=<6.2047 1.7395 0.0159>, rotation=<rot 0 -90.0004 -90.1466>}]
}
constraints [
 {bone1Name=dyn_hair_02, bone2Name=dyn_hair_01, swing1LimitDegrees=4, swing2LimitDegrees=4, twistLimitDegrees=4}
 {bone1Name=dyn_hair_03, bone2Name=dyn_hair_02, swing1LimitDegrees=4, swing2LimitDegrees=4, twistLimitDegrees=4}
 {bone1Name=dyn_hair_04, bone2Name=dyn_hair_03, swing1LimitDegrees=4, swing2LimitDegrees=4, twistLimitDegrees=4}
 {bone1Name=dyn_hair_05, bone2Name=dyn_hair_04, swing1LimitDegrees=4, swing2LimitDegrees=4, twistLimitDegrees=4}
 {bone1Name=dyn_hair_01, bone2Name=head, swing1LimitDegrees=4, swing2LimitDegrees=4, twistLimitDegrees=4}
]
